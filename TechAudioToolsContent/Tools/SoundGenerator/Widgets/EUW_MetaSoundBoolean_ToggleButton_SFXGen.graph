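# Editor utility toggle button for a boolean MetaSound input in the TechAudioTools Sound Generator, wrapping a primitive bool toggle with a content slot.
# It reports its Input Name, then casts the matching supplied viewmodel to MetaSoundInputEditorViewModel and assigns it.
# PreConstruct applies the checked or unchecked style from Default Value.
# Helper functions supply icon color, hide reset at the default value, and expose hover state, broadcast as a FieldNotify on mouse enter/leave.
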EditorUtilityWidgetBlueprint /TechAudioToolsContent/Tools/SoundGenerator/Widgets/EUW_MetaSoundBoolean_ToggleButton_SFXGen
graphs [
 EventGraph {nodes [
   n1 = Event OnMouseEnter()
   Broadcast Field Value Changed!(execute=[@n1.then, @n2.then], self:Widget Object Reference, FieldId:Field Notification Id Structure=(FieldName="Is Hovered")) -> (then)
   n2 = Event OnMouseLeave!() -> (OutputDelegate, then, MouseEvent:"Pointer Event Structure (by ref)") {is=title K2Node_Event}
   Event PreConstruct()
   > Set with Broadcast Check Box Style!(
    execute
    Check Box Style:Check Box Style Structure=Get Toggle Style!(self, bIsChecked:bool=Get Default Value() -> (Default Value:bool)) -> (CheckBoxStyleOut:Check Box Style Structure, IconLinearColor:Linear Color Structure) {is=title K2Node_CallFunction}
    self:Meta Sound Literal Bool Primitive Toggle Button Object Reference=Get Bool_ToggleButton() -> (Bool_ToggleButton:Meta Sound Literal Bool Primitive Toggle Button Object Reference)
   ) -> (then, Output_Get:Check Box Style Structure)
   n3 = Event Set Input Viewmodels!() -> (OutputDelegate, then, InputViewModels:Map of Names to Meta Sound Input Viewmodel Object References) {is=title K2Node_Event}
   > n4 = Cast To MetaSoundInputEditorViewModel!(
    execute
    Object=Find(
     TargetMap:Map of Names to Meta Sound Input Viewmodel Object References=@n3.InputViewModels
     Key:"Name (by ref)"=Get Input Name() -> (Input Name:Name=None)
    ) -> (Value:Meta Sound Input Viewmodel Object Reference)
   ) -> (then, CastFailed, AsMetaSound Input Editor Viewmodel:Meta Sound Input Editor Viewmodel Object Reference) {is=title K2Node_DynamicCast}
   > SetMetaSoundInputEditorViewModel!(execute, self, Viewmodel:Meta Sound Input Editor Viewmodel Object Reference=@n4) -> (then)
  ]}
 Get Reset Button Visibility {nodes [
   n1 = Get Reset Button Visibility!() -> (then, Value:bool=~) {is=title K2Node_FunctionEntry}
   n2 = Get Default Value()
   > Return Node!(
    execute
    ReturnValue:ESlateVisibility Enum=Select(
     Option 0=Visible
     Option 1=Hidden
     Index="Equal (Boolean)"(A:bool=@n1, B:bool=@n2) -> (ReturnValue:bool=~) {is=title K2Node_PromotableOperator}
    )
   )
  ]}
 Get Toggle Style {nodes [
   n1 = Get Toggle Style!() -> (then, bIsChecked:bool=~) {is=title K2Node_FunctionEntry}
   > Return Node!(
    execute
    CheckBoxStyleOut:Check Box Style Structure=Select(
     Option 0:Check Box Style Structure=Get UncheckedStyle() -> (UncheckedStyle:Check Box Style Structure)
     Option 1:Check Box Style Structure=Get CheckedStyle() -> (CheckedStyle:Check Box Style Structure)
     Index=@n1
    ) -> (ReturnValue:Check Box Style Structure)
    IconLinearColor:Linear Color Structure=Select(Option 0:Linear Color Structure=(R=0.2086,G=0.2086,B=0.2086,A=1), Option 1:Linear Color Structure=(R=1,G=0,B=0,A=1), Index=@n1) -> (ReturnValue:Linear Color Structure)
   )
  ]}
 GetInputViewModelNames {nodes [
   Get Input Viewmodel Names() {is=title K2Node_FunctionEntry}
   > Return Node!(execute, ReturnValue:Array of Names=Make Array("[0]":Name=Get Input Name() -> (Input Name=None)) -> (Array:Array of Names))
  ]}
 Is Hovered {nodes [
   Is Hovered() {is=title K2Node_FunctionEntry}
   > Return Node(ReturnValue:bool=IsHovered())
  ]}
]
variables [
 bIsChecked {type=boolean}
 Default Value {type=boolean}
 Input Name {type=string, default=None}
 Label Text {type=string, default=Label}
]
widgets {
 info {
  namedSlotCount=1
  rootWidgetClass=/TechAudioToolsContent/Widgets/MetaSoundLiterals/Bool/Primitives/WBP_MetaSoundLiteralBoolPrimitive_ToggleButton.WBP_MetaSoundLiteralBoolPrimitive_ToggleButton_C
  widgetCount=2
 }
 widgets [
  Bool_ToggleButton {
   widgetClassPath=/TechAudioToolsContent/Widgets/MetaSoundLiterals/Bool/Primitives/WBP_MetaSoundLiteralBoolPrimitive_ToggleButton
   bIsVariable=true
   widget=Bool_ToggleButton
  }
  Content {widgetClassPath=&NamedSlot, namedSlotHost=Bool_ToggleButton, widget=Content}
 ]
}
