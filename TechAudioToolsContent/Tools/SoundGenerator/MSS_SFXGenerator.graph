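# Parameter-driven one-shot mono MetaSound SFX generator that synthesizes sounds from five switchable waveforms.
# Pitch jumps, sweep and vibrato drive the selected oscillator through serial waveshaper, bitcrusher, ring mod and filter patches.
# An ADSR envelope with LFO modulation shapes amplitude before dry, delay, reverb and flanger signals are mixed.
# Audio buses expose intermediate stages, a Wave Writer optionally records the mix, and an envelope-follower threshold fires OnFinished.
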MetaSoundSource /TechAudioToolsContent/Tools/SoundGenerator/MSS_SFXGenerator
inputs [
 Amp.BitCrusher.SampleRate {type=Float}
 Amp.Envelope.Attack {type=Time}
 Amp.Envelope.Decay {type=Time}
 Amp.Envelope.Duration {type=Time}
 Amp.Envelope.Release {type=Time}
 Amp.Envelope.Sustain {type=Float}
 Amp.LFO.Depth {type=Float}
 Amp.LFO.Frequency {type=Float}
 FX.BitCrusher.Amount {type=Float}
 FX.Delay.Amount {type=Float}
 FX.Delay.Feedback {type=Float}
 FX.Delay.Time {type=Float}
 FX.Filter.Cutoff {type=Float}
 FX.Filter.Envelope.Curve {type=Float}
 FX.Filter.Envelope.Depth {type=Float}
 FX.Filter.Envelope.Shape {type=Float}
 FX.Filter.Resonance {type=Float}
 FX.Filter.Type {type=Float}
 FX.Flanger.Amount {type=Float}
 FX.Flanger.Delay {type=Float}
 FX.Reverb.Amount {type=Float}
 FX.Ringmod.Amount {type=Float}
 FX.Ringmod.Envelope.Curve {type=Float}
 FX.Ringmod.Envelope.Depth {type=Float}
 FX.Ringmod.Envelope.Shape {type=Float}
 FX.RingmodMod.Frequency {type=Float}
 FX.WaveShaper.Amount {type=Float}
 Generator.Pitch.BaseFrequency {type=Float}
 Generator.Pitch.Jump.Num {type=Int32}
 Generator.Pitch.Jump.Size {type=Float}
 Generator.Pitch.Jump.Speed {type=Float}
 Generator.Pitch.Sweep.Curve {type=Float}
 Generator.Pitch.Sweep.Target {type=Float}
 Generator.PW.DutyCycle {type=Float}
 Generator.PW.DutySweep {type=Float}
 Generator.Vibrato.Depth {type=Float}
 Generator.Vibrato.Frequency {type=Float}
 Generator.Waveform {type=Float}
 UE.Source.OnPlay {type=Trigger}
 WaveWriter.Enabled {type=Bool}
 WaveWriter.FileName {type=String}
]
interfaces [UE.OutputFormat.Mono:1.0, UE.Source.OneShot:1.0, UE.Source:1.1]
outputs [
 Amp.Envelope.Value {type=Float}
 Amp.LFO.Value {type=Float}
 Amp.Value {type=Float}
 "UE.OutputFormat.Mono.Audio:0" {type=Audio}
 UE.Source.OneShot.OnFinished {type=Trigger}
]
pages [Default {
  input_defaults {
   Amp.BitCrusher.SampleRate=0
   Amp.Envelope.Attack=0
   Amp.Envelope.Decay=0
   Amp.Envelope.Duration=0
   Amp.Envelope.Release=1
   Amp.Envelope.Sustain=1
   Amp.LFO.Depth=0
   Amp.LFO.Frequency=0
   FX.BitCrusher.Amount=0
   FX.Delay.Amount=0
   FX.Delay.Feedback=0
   FX.Delay.Time=0
   FX.Filter.Cutoff=1
   FX.Filter.Envelope.Curve=0
   FX.Filter.Envelope.Depth=0
   FX.Filter.Envelope.Shape=0
   FX.Filter.Resonance=0
   FX.Filter.Type=0
   FX.Flanger.Amount=0
   FX.Flanger.Delay=0
   FX.Reverb.Amount=0
   FX.Ringmod.Amount=0
   FX.Ringmod.Envelope.Curve=0
   FX.Ringmod.Envelope.Depth=0
   FX.Ringmod.Envelope.Shape=0
   FX.RingmodMod.Frequency=0
   FX.WaveShaper.Amount=0
   Generator.PW.DutyCycle=0.5
   Generator.PW.DutySweep=0
   Generator.Pitch.BaseFrequency=0.256
   Generator.Pitch.Jump.Num=0
   Generator.Pitch.Jump.Size=0
   Generator.Pitch.Jump.Speed=0
   Generator.Pitch.Sweep.Curve=0.5
   Generator.Pitch.Sweep.Target=0.5
   Generator.Vibrato.Depth=0
   Generator.Vibrato.Frequency=0
   Generator.Waveform=0
   UE.Source.OnPlay=null
   WaveWriter.Enabled=null
   WaveWriter.FileName=SFXGenerator
  }
  nodes [
   LinearToLogFrequency_0 = UE.Linear To Log Frequency(Value=^Generator.Pitch.BaseFrequency) {also {defaults {Max Range=6000, Min Range=60}}}
   n1 = MapRange.MapRange:Float(In=UE.Power:Float(PrimaryOperand=UE.ConversionTimeToFloat(In=^Amp.Envelope.Duration).Out) {also {defaults {AdditionalOperands=2}}}.Out) {also {defaults {Out Range A=0.01, Out Range B=5}}}
   n2 = MapRange.MapRange:Float(In=UE.Power:Float(PrimaryOperand=UE.ConversionTimeToFloat(In=^Amp.Envelope.Decay).Out) {also {defaults {AdditionalOperands=2}}}.Out) {also {defaults {Out Range B=2}}}
   n3 = MapRange.MapRange:Float(In=UE.Power:Float(PrimaryOperand=UE.ConversionTimeToFloat(In=^Amp.Envelope.Attack).Out) {also {defaults {AdditionalOperands=2}}}.Out) {also {defaults {Out Range B=2}}}
   n4 = Max.Max:Float(
    A=UE.Reroute(Value=@n1.Out Value).Value
    B=UE.Reroute(Value=UE.Add:Float(AdditionalOperands=@n2.Out Value, PrimaryOperand=@n3.Out Value).Out).Value
   )
   n5 = VariableAccessor.Time(Variable=VariableMutator.Time(Value=UE.ConversionFloatToTime(In=@n4.Value).Out, Variable=InitVariable.Time().Variable).Variable)
   n6 = /TechAudioToolsContent/Tools/SoundGenerator/Patches/MSP_SG_MultiStageJump(NumberOfStages=^Generator.Pitch.Jump.Num, PitchInterval=^Generator.Pitch.Jump.Size, OnPlay=^UE.Source.OnPlay, StageDuration=@n5.Value)
   n7 = VariableAccessor.Float(Variable=VariableMutator.Float(
     Variable=InitVariable.Float().Variable
     Value=UE.Multiply:Float(PrimaryOperand=@LinearToLogFrequency_0.Frequency, AdditionalOperands=@n6.Value).Out
    ).Variable)
   n8 = VariableAccessor.Float(Variable=@n7.Variable)
   n9 = UE.ConversionFloatToTime(In=@n1.Out Value)
   n10 = UE.ConversionTimeToFloat(In=^Amp.Envelope.Release)
   n11 = UE.ConversionFloatToTime(In=MapRange.MapRange:Float(In=UE.Power:Float(PrimaryOperand=@n10.Out) {also {defaults {AdditionalOperands=2}}}.Out) {also {defaults {Out Range B=5}}}.Out Value)
   n12 = VariableAccessor.Time(Variable=VariableMutator.Time(Value=UE.Add:Time(PrimaryOperand=@n9.Out, AdditionalOperands=@n11.Out).Out, Variable=InitVariable.Time().Variable).Variable)
   n13 = VariableAccessor.Time(Variable=@n12.Variable)
   n14 = VariableAccessor.Time(Variable=@n13.Variable)
   n15 = /TechAudioToolsContent/Tools/SoundGenerator/Patches/MSP_SG_FM_PitchSlide(
    PitchCurve=^Generator.Pitch.Sweep.Curve
    PitchTarget=^Generator.Pitch.Sweep.Target
    OnPlay=^UE.Source.OnPlay
    BaseFrequency=@n8.Value
    Duration=VariableAccessor.Time(Variable=@n14.Variable).Value
   )
   n16 = /TechAudioToolsContent/Tools/SoundGenerator/Patches/MSP_SG_FM_Vibrato(
    Depth=^Generator.Vibrato.Depth
    VibratoFrequency=^Generator.Vibrato.Frequency
    BaseFrequency=VariableAccessor.Float(Variable=@n8.Variable).Value
   )
   n17 = UE.Add:Audio(PrimaryOperand=@n15.Out Value, AdditionalOperands=@n16.Out)
   TriggerSelect5_0 = "UE.Trigger Select (5)"(
    In=^UE.Source.OnPlay
    Index=Convert.Int32:Float(Float=UE.Multiply:Float(PrimaryOperand=^Generator.Waveform) {also {defaults {AdditionalOperands=4.9}}}.Out).Int32
   )
   TriggerRouteTriggerRouteAudio5_0 = "TriggerRoute.Trigger Route (Audio, 5):Audio"(
    Value 4=/TechAudioToolsContent/Tools/SoundGenerator/Patches/MSP_SG_Generator_Noise(Glide=^Generator.Pitch.Jump.Speed, FM=@n17.Out, Frequency=@n7.Value).Output
    Value 2=/TechAudioToolsContent/Tools/SoundGenerator/Patches/MSP_SG_Generator_Saw(Glide=^Generator.Pitch.Jump.Speed, FM=@n17.Out, Frequency=@n7.Value).Output
    Value 3=/TechAudioToolsContent/Tools/SoundGenerator/Patches/MSP_SG_Generator_Sine(Glide=^Generator.Pitch.Jump.Speed, FM=@n17.Out, Frequency=@n7.Value).Output
    Value 0=/TechAudioToolsContent/Tools/SoundGenerator/Patches/MSP_SG_Generator_Square(
     Duty=^Generator.PW.DutyCycle
     DutySweep=^Generator.PW.DutySweep
     Glide=^Generator.Pitch.Jump.Speed
     FM=@n17.Out
     Frequency=@n7.Value
     Duration=@n12.Value
    ).Output
    Value 1=/TechAudioToolsContent/Tools/SoundGenerator/Patches/MSP_SG_Generator_Triangle(Glide=^Generator.Pitch.Jump.Speed, FM=@n17.Out, Frequency=@n7.Value).Output
    Set 0=@TriggerSelect5_0.Out 0
    Set 1=@TriggerSelect5_0.Out 1
    Set 2=@TriggerSelect5_0.Out 2
    Set 3=@TriggerSelect5_0.Out 3
    Set 4=@TriggerSelect5_0.Out 4
   )
   AudioBusWriter1_0 = "UE.Audio Bus Writer (1)"(In 0=@TriggerRouteTriggerRouteAudio5_0.Value) {also {defaults {Audio Bus=/TechAudioToolsContent/Tools/SoundGenerator/AudioBuses/AB_SFXG_Osc}}}
   n18 = Max.Max:Float(A=@n15.Range) {also {defaults {B=1}}}
   AudioBusWriter1_1 = "UE.Audio Bus Writer (1)"(In 0=UE.Multiply:Audio by Float(AdditionalOperands=UE.Divide:Float(AdditionalOperands=@n18.Value).Out, PrimaryOperand=@n15.Out Value).Out) {also {defaults {Audio Bus=/TechAudioToolsContent/Tools/SoundGenerator/AudioBuses/AB_SFXG_PitchBend}}}
   n19 = /TechAudioToolsContent/Tools/SoundGenerator/Patches/MSP_SG_FX_InlineWaveShaper(Amount=^FX.WaveShaper.Amount, AudioIn=@TriggerRouteTriggerRouteAudio5_0.Value)
   n20 = /TechAudioToolsContent/Tools/SoundGenerator/Patches/MSP_SG_FX_InlineBitcrusher(SampleRate=^Amp.BitCrusher.SampleRate, Amount=^FX.BitCrusher.Amount, AudioIn=@n19.Out)
   n21 = /TechAudioToolsContent/Tools/SoundGenerator/Patches/MSP_SG_FX_InlineRingmod(
    Amount=^FX.Ringmod.Amount
    EnvelopeCurve=^FX.Ringmod.Envelope.Curve
    EnvelopeDepth=^FX.Ringmod.Envelope.Depth
    EnvelopeShape=^FX.Ringmod.Envelope.Shape
    Frequency=^FX.RingmodMod.Frequency
    Trigger=^UE.Source.OnPlay
    AudioIn=@n20.Out
    Duration=@n14.Value
   )
   AudioBusWriter1_2 = "UE.Audio Bus Writer (1)"(In 0=@n21.Out) {also {defaults {Audio Bus=/TechAudioToolsContent/Tools/SoundGenerator/AudioBuses/AB_SFXG_RingMod}}}
   n22 = /TechAudioToolsContent/Tools/SoundGenerator/Patches/MSP_SG_FX_InlineFilter(
    FilterCutoff=^FX.Filter.Cutoff
    FilterEnvelopeCurve=^FX.Filter.Envelope.Curve
    FilterEnvelopeDepth=^FX.Filter.Envelope.Depth
    FilterEnvelopeShape=^FX.Filter.Envelope.Shape
    FilterResonance=^FX.Filter.Resonance
    FilterType=^FX.Filter.Type
    TriggerEnvelope=^UE.Source.OnPlay
    AudioIn=UE.Reroute(Value=@n21.Out).Value
    Duration=@n13.Value
   )
   n23 = VariableAccessor.Time(Variable=@n5.Variable)
   TriggerDelay_0 = UE.Trigger Delay(In=^UE.Source.OnPlay, Delay Time=@n23.Value)
   n24 = UE.ConversionFloatToTime(In=@n3.Out Value)
   n25 = InitVariable.Time()
   n26 = InitVariable.Time()
   n27 = UE.ConversionFloatToTime(In=@n2.Out Value)
   n28 = InitVariable.Time()
   ADSREnvelopeADSREnvelope_0 = ADSR Envelope.ADSR Envelope:Float(
    Sustain Level=^Amp.Envelope.Sustain
    Trigger Attack=^UE.Source.OnPlay
    Trigger Release=@TriggerDelay_0.Out
    Attack Time=VariableAccessor.Time(Variable=VariableMutator.Time(Value=@n24.Out, Variable=@n25.Variable).Variable).Value
    Release Time=VariableAccessor.Time(Variable=VariableMutator.Time(Value=@n11.Out, Variable=@n26.Variable).Variable).Value
    Decay Time=VariableAccessor.Time(Variable=VariableMutator.Time(Value=@n27.Out, Variable=@n28.Variable).Variable).Value
   )
   n29 = /TechAudioToolsContent/Tools/SoundGenerator/Patches/MSP_SG_AM_LFO(Depth=^Amp.LFO.Depth, Frequency=^Amp.LFO.Frequency, AmpSignal=@ADSREnvelopeADSREnvelope_0.Out Envelope)
   n30 = VariableAccessor.Float(Variable=VariableMutator.Float(Variable=InitVariable.Float().Variable, Value=@n29.Out).Variable)
   n31 = VariableAccessor.Audio(Variable=VariableMutator.Audio(Variable=InitVariable.Audio().Variable, Value=UE.Multiply:Audio by Float(PrimaryOperand=@n22.Out, AdditionalOperands=@n30.Value).Out).Variable)
   n32 = VariableAccessor.Audio(Variable=@n31.Variable)
   n33 = VariableAccessor.Audio(Variable=@n32.Variable)
   n34 = /TechAudioToolsContent/Tools/SoundGenerator/Patches/MSP_SG_FX_SendReverb(Amount=^FX.Reverb.Amount, AudioIn=@n33.Value)
   AudioBusWriter1_3 = "UE.Audio Bus Writer (1)"(In 0=@n34.Out) {also {defaults {Audio Bus=/TechAudioToolsContent/Tools/SoundGenerator/AudioBuses/AB_SFXG_Reverb}}}
   AudioBusWriter1_4 = "UE.Audio Bus Writer (1)"(In 0=@n19.Out) {also {defaults {Audio Bus=/TechAudioToolsContent/Tools/SoundGenerator/AudioBuses/AB_SFXG_WaveShaper}}}
   AudioBusWriter1_5 = "UE.Audio Bus Writer (1)"(In 0=@n20.Out) {also {defaults {Audio Bus=/TechAudioToolsContent/Tools/SoundGenerator/AudioBuses/AB_SFXG_BitCrusher}}}
   AudioBusWriter1_6 = "UE.Audio Bus Writer (1)"(In 0=UE.Multiply:Audio by Float(AdditionalOperands=^Generator.Vibrato.Depth, PrimaryOperand=@n16.NormalizedOut).Out) {also {defaults {Audio Bus=/TechAudioToolsContent/Tools/SoundGenerator/AudioBuses/AB_SFXG_Vibrato}}}
   n35 = /TechAudioToolsContent/Tools/SoundGenerator/Patches/MSP_SG_FX_SendDelay(
    DelayAmount=^FX.Delay.Amount
    DelayFeedback=^FX.Delay.Feedback
    DelayTime=^FX.Delay.Time
    AudioIn=@n31.Value
    Duration=VariableAccessor.Time(Variable=@n23.Variable).Value
   )
   AudioBusWriter1_7 = "UE.Audio Bus Writer (1)"(In 0=@n35.Out) {also {defaults {Audio Bus=/TechAudioToolsContent/Tools/SoundGenerator/AudioBuses/AB_SFXG_Delay}}}
   AudioBusWriter1_8 = "UE.Audio Bus Writer (1)"(In 0=@n22.Out) {also {defaults {Audio Bus=/TechAudioToolsContent/Tools/SoundGenerator/AudioBuses/AB_SFXG_Filter}}}
   Multiply_5 = UE.Multiply:Float(AdditionalOperands=^Amp.LFO.Depth, PrimaryOperand=@n29.LFO.Output)
   AudioMixerAudioMixerMono4_0 = "AudioMixer.Audio Mixer (Mono, 4)"(
    In 1=@n35.Out
    In 3=/TechAudioToolsContent/Tools/SoundGenerator/Patches/MSP_SG_FX_SendFlanger(Amount=^FX.Flanger.Amount, Delay=^FX.Flanger.Delay, AudioIn=VariableAccessor.Audio(Variable=@n33.Variable).Value).Out
    In 2=@n34.Out
    In 0=@n32.Value
   )
   EnvelopeFollower_0 = UE.Envelope Follower(In=@AudioMixerAudioMixerMono4_0.Out)
   TriggerOnThreshold_0 = UE.TriggerOnThreshold:Audio(In=@EnvelopeFollower_0.Audio Envelope) {also {defaults {Threshold=0.01, Type=1}}}
   UE.WaveWriter:Audio(Enabled=^WaveWriter.Enabled, Filename Prefix=^WaveWriter.FileName, In=@AudioMixerAudioMixerMono4_0.Out) {also {version [1,1]}}
  ]
  links [
   {dst="@Amp.Envelope.Value", src=ADSREnvelopeADSREnvelope_0.Out Envelope}
   {dst="@Amp.Value", src=ADSREnvelopeADSREnvelope_0.Out Envelope}
   {dst="@UE.OutputFormat.Mono.Audio:0", src=AudioMixerAudioMixerMono4_0.Out}
   {dst="@Amp.LFO.Value", src=Multiply_5.Out}
   {dst="@UE.Source.OneShot.OnFinished", src=TriggerOnThreshold_0.Out}
  ]
 }]
properties {assetImportData=&self:AssetImportData, sampleRate=48000}
