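# EUW_Button_Simple_Icon is a reusable editor utility button widget that shows a configurable Slate icon for TechAudioTools editor UI.
# It wraps an EUW_Button_Simple with the icon in its content slot and re-broadcasts its clicked, pressed and released events as its own dispatchers.
# PreConstruct applies the icon brush path and size variables.
# Icon color switches between default and highlighted based on hover, refreshed on pre-construct, mouse enter, mouse leave and click.
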EditorUtilityWidgetBlueprint /TechAudioToolsContent/Widgets/Editor/Buttons/EUW_Button_Simple_Icon
graphs [
 EventGraph {nodes [
   "On Clicked (Button_ContentSlot)"() {is=title K2Node_ComponentBoundEvent}
   > Call On Clicked!(execute:Exec, self:Tech Audio Tools Button Simple Icon Object Reference) -> (then:Exec) {is=title K2Node_CallDelegate}
   > Update Icon Color()
   "On Pressed (Button_ContentSlot)"() {is=title K2Node_ComponentBoundEvent}
   > Call On Pressed!(execute:Exec, self:Tech Audio Tools Button Simple Icon Object Reference) -> (then:Exec) {is=title K2Node_CallDelegate}
   "On Released (Button_ContentSlot)"() {is=title K2Node_ComponentBoundEvent}
   > Call On Released!(execute:Exec, self:Tech Audio Tools Button Simple Icon Object Reference) -> (then:Exec) {is=title K2Node_CallDelegate}
   Update Icon Color() {is=title K2Node_CustomEvent}
   > SetIconColor!(
    execute
    self:Tech Audio Tools Editor Slate Icon Object Reference=Get Icon() -> (Icon:Tech Audio Tools Editor Slate Icon Object Reference)
    InColor:Linear Color Structure=SelectColor(
     A=Get Highlighted Icon Color() -> (Highlighted Icon Color:Linear Color Structure)
     B=Get Default Icon Color() -> (Default Icon Color:Linear Color Structure)
     bPickA=IsHovered()
    )
   ) -> (then)
   Event PreConstruct()
   > SetIconBrushByPath!(
    execute
    self:Tech Audio Tools Editor Slate Icon Object Reference=Get Icon()
    InFullIconPath:Name=Get Icon Brush Path() -> (Icon Brush Path:Name=None)
   ) -> (then)
   > SetIconSize!(
    execute
    self:Tech Audio Tools Editor Slate Icon Object Reference=Get Icon()
    InSize:Vector 2D Structure=Get Icon Size() -> (Icon Size:Vector 2D Structure)
   ) -> (then)
   > Update Icon Color() {is=title K2Node_CallFunction}
   Event OnMouseEnter()
   > Update Icon Color()
   Event OnMouseLeave!() -> (OutputDelegate, then, MouseEvent:"Pointer Event Structure (by ref)") {is=title K2Node_Event}
   > Update Icon Color()
  ]}
 On Clicked {nodes [On Clicked() {is=title K2Node_FunctionEntry}]}
 On Pressed {nodes [On Pressed() {is=title K2Node_FunctionEntry}]}
 On Released {nodes [On Released() {is=title K2Node_FunctionEntry}]}
]
variables [
 Default Icon Color {type=LinearColor, default=<rgba 0.5271 0.5271 0.5271 1>}
 Highlighted Icon Color {type=LinearColor, default=<rgba 1 1 1 1>}
 Icon Brush Path {type=string, default=EditorStyle.Icons.Save}
 Icon Size {type=Vector2D, default=<20 20>}
 On Clicked {default=()}
 On Pressed {default=()}
 On Released {default=()}
]
widgets {
 info {namedSlotCount=1, rootWidgetClass=&/EUW_Button_Simple.EUW_Button_Simple_C, widgetCount=3}
 widgets [
  Button_ContentSlot {widgetClassPath=&/EUW_Button_Simple, bIsVariable=true, widget=Button_ContentSlot}
  VerticalBox_0 {widgetClassPath=&VerticalBox, namedSlotHost=Button_ContentSlot, widget=VerticalBox_0}
  Icon {
   widgetClassPath=/Script/TechAudioToolsContentEditor.TechAudioToolsEditorSlateIcon
   bIsVariable=true
   parent=VerticalBox_0
   widget=Icon
   slot {class=VerticalBoxSlot, properties {padding=<ltrb 3 4 3 4>}}
  }
 ]
}
