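# Editor utility widget for the SFX Generator tool that pairs a randomize button with a lock toggle in a horizontal row.
# Clicking the button broadcasts the Randomize event dispatcher, but only while the lock is unchecked.
# Toggling the lock swaps its icon between locked and unlocked and dims or restores the randomize button.
# GetIsLocked and SetLockedState expose the lock state, and PreConstruct applies a shared IconSize to both icons.
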EditorUtilityWidgetBlueprint /TechAudioToolsContent/Tools/SoundGenerator/Widgets/EUW_SFXGenerator_Randomizer
graphs [
 EventGraph {nodes [
   "On Clicked (EUW_Button_Simple_Icon)"() {is=title K2Node_ComponentBoundEvent}
   n1 = Get EUW_ToggleButton_Icon_SFXGen() -> (EUW_ToggleButton_Icon_SFXGen:EUW Toggle Button Icon SFXGen Object Reference)
   > Branch(Condition=NOT Boolean(A=Get Is Checked!(self:EUW Toggle Button Icon SFXGen Object Reference=@n1) -> (Return Value:bool) {is=title K2Node_CallFunction}))
   > Call Randomize!(execute:Exec, self:EUW SFXGenerator Randomizer Object Reference) -> (then:Exec) {is=title K2Node_CallDelegate}
   n2 = "On Checked State Changed (EUW_ToggleButton_Icon_SFXGen)"!() -> (OutputDelegate, then, Is Checked:bool) {is=title K2Node_ComponentBoundEvent}
   > n3 = Branch(Condition=@n2.Is Checked)
   n4 = Get EUW_ToggleButton_Icon_SFXGen()
   > Set with Broadcast Icon Brush Path!(execute, Icon Brush Path:Name=EditorStyle.Icons.lock, self:EUW Toggle Button Icon SFXGen Object Reference=@n4) -> (then, Output_Get:Name=None)
   n5 = Get EUW_Button_Simple_Icon() -> (EUW_Button_Simple_Icon:Tech Audio Tools Button Simple Icon Object Reference)
   > SetColorAndOpacity(self:User Widget Object Reference=@n5, InColorAndOpacity:Linear Color Structure=(R=1,G=1,B=1,A=0.2))
   Set with Broadcast Icon Brush Path(execute=@n3.else, Icon Brush Path=EditorStyle.Icons.unlock, self=@n4) -> (Output_Get=None)
   > SetColorAndOpacity(self=@n5, InColorAndOpacity=(R=1,G=1,B=1,A=1))
   Event PreConstruct()
   > Set with Broadcast Icon Size!(
    execute
    Icon Size:Vector 2D Structure=Get IconSize() -> (IconSize:Vector 2D Structure)
    self:EUW Toggle Button Icon SFXGen Object Reference=Get EUW_ToggleButton_Icon_SFXGen()
   ) -> (then, Output_Get:Vector 2D Structure)
   > Set with Broadcast Icon Size(Icon Size=Get IconSize(), self:Tech Audio Tools Button Simple Icon Object Reference=Get EUW_Button_Simple_Icon())
   Event Tick()
  ]}
 GetIsLocked {nodes [
   GetIsLocked() {is=title K2Node_FunctionEntry}
   > Return Node!(execute, bIsChecked:bool=Get Is Checked(self=Get EUW_ToggleButton_Icon_SFXGen()))
  ]}
 Randomize {nodes [Randomize() {is=title K2Node_FunctionEntry}]}
 SetLockedState {nodes [
   n1 = SetLockedState!() -> (then, bIsLocked:bool=~) {is=title K2Node_FunctionEntry}
   > Set Checked State!(execute, self:EUW Toggle Button Icon SFXGen Object Reference=Get EUW_ToggleButton_Icon_SFXGen(), Is Checked:bool=@n1) -> (then) {is=title K2Node_CallFunction}
  ]}
]
variables [
 IconSize {type=Vector2D, default=<20 20>}
 Randomize {default=()}
]
widgets {
 info {rootWidgetClass=&HorizontalBox, widgetCount=3}
 widgets [
  HorizontalBox_0 {widgetClassPath=&HorizontalBox, widget=HorizontalBox_0}
  EUW_Button_Simple_Icon {
   widgetClassPath=/TechAudioToolsContent/Widgets/Editor/Buttons/EUW_Button_Simple_Icon
   bIsVariable=true
   parent=HorizontalBox_0
   widget=EUW_Button_Simple_Icon
   properties {icon Brush Path=EditorStyle.Icons.Convert}
   slot {class=HorizontalBoxSlot}
  }
  EUW_ToggleButton_Icon_SFXGen {
   widgetClassPath=&/EUW_ToggleButton_Icon_SFXGen
   bIsVariable=true
   parent=HorizontalBox_0
   widget=EUW_ToggleButton_Icon_SFXGen
   properties {icon Brush Path=EditorStyle.Icons.unlock}
   slot {class=HorizontalBoxSlot}
  }
 ]
}
