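# Reusable UMG panel widget with a rounded border wrapping a header named slot above a fill-sized body named slot.
# PreConstruct applies the Background Color to the border and collapses the header when Hide Header is set.
# If Hover Outline Color is non-zero, mouse enter caches the border brush and applies a copy with that outline color.
# Mouse leave restores the cached brush, using the same color check.
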WidgetBlueprint /TechAudioToolsContent/Widgets/Layout/WBP_HeaderContentPanel
graphs [
 Create Hover Outline Brush {nodes [
   n1 = Create Hover Outline Brush!() -> (then, Base Brush:"Slate Brush Structure (by ref)", Hover Outline Color:Linear Color Structure) {is=title K2Node_FunctionEntry}
   n2 = Break Slate Brush!(SlateBrush:"Slate Brush Structure (by ref)"=@n1) -> (
    TintColor:Slate Color Structure
    DrawAs:ESlateBrushDrawType Enum
    Tiling:ESlateBrushTileType Enum
    Mirroring:ESlateBrushMirrorType Enum
    ImageSize:Vector 2D Structure
    Margin:Margin Structure
    ResourceObject:Object Reference
    OutlineSettings:Slate Brush Outline Settings Structure
   ) {is=title K2Node_BreakStruct}
   n3 = Break Slate Brush Outline Settings!(SlateBrushOutlineSettings:"Slate Brush Outline Settings Structure (by ref)"=@n2.OutlineSettings) -> (
    CornerRadii:Vector 4 Structure
    Color:Slate Color Structure
    Width:float=~
    RoundingType:ESlateBrushRoundingType Enum
    bUseBrushTransparency:bool=~
   ) {is=title K2Node_BreakStruct}
   n4 = Make Slate Color(SpecifiedColor=@n1.Hover Outline Color)
   > Return Node!(
    execute
    SlateBrush:Slate Brush Structure=Make Slate Brush!(
     TintColor:Slate Color Structure=@n2
     DrawAs:ESlateBrushDrawType Enum=@n2.DrawAs
     Tiling:ESlateBrushTileType Enum=@n2.Tiling
     Mirroring:ESlateBrushMirrorType Enum=@n2.Mirroring
     ImageSize:Vector 2D Structure=@n2.ImageSize
     Margin:Margin Structure=@n2.Margin
     ResourceObject:Object Reference=@n2.ResourceObject
     OutlineSettings:Slate Brush Outline Settings Structure=Make Slate Brush Outline Settings!(
      CornerRadii:Vector 4 Structure=@n3
      Color:Slate Color Structure=@n4
      Width:float=@n3.Width
      RoundingType:ESlateBrushRoundingType Enum=@n3.RoundingType
      bUseBrushTransparency:bool=@n3.bUseBrushTransparency
     ) -> (SlateBrushOutlineSettings:Slate Brush Outline Settings Structure) {is=title K2Node_MakeStruct}
    ) -> (SlateBrush:Slate Brush Structure) {is=title K2Node_MakeStruct}
   )
  ]}
 EventGraph {nodes [
   Event PreConstruct()
   > SetBrushColor(self=Get Border() -> (Border:Border Object Reference), InBrushColor=Get Background Color() -> (Background Color:Linear Color Structure))
   > SetVisibility(
    self=Get HeaderSlot() -> (HeaderSlot:Named Slot Object Reference)
    InVisibility=Select(Option 0=Visible, Option 1=Collapsed, Index=Get Hide Header() -> (Hide Header:bool))
   )
   Event OnMouseEnter()
   > Branch(Condition="Not Equal (LinearColor)"(
     A:Linear Color Structure=Get Hover Outline Color() -> (Hover Outline Color:Linear Color Structure)
     B:Linear Color Structure=(R=0,G=0,B=0,A=0)
    ) -> (ReturnValue:bool=~) {is=title K2Node_PromotableOperator})
   > n1 = Set Brush Cache(Brush Cache:Slate Brush Structure=Get Background!(self:Border Object Reference=Get Border()) -> (Background:Slate Brush Structure)) -> (Output_Get:Slate Brush Structure)
   > SetBrush(
    self:Border Object Reference=Get Border()
    InBrush=Create Hover Outline Brush!(self, Base Brush:"Slate Brush Structure (by ref)"=@n1, Hover Outline Color:Linear Color Structure=Get Hover Outline Color()) -> (SlateBrush:Slate Brush Structure) {is=title K2Node_CallFunction}
   )
   Event OnMouseLeave!() -> (OutputDelegate, then, MouseEvent:"Pointer Event Structure (by ref)") {is=title K2Node_Event}
   > Branch(Condition="Not Equal (LinearColor)"(A=Get Hover Outline Color(), B=(R=0,G=0,B=0,A=0)) -> (ReturnValue=~))
   > SetBrush(self=Get Border(), InBrush=Get Brush Cache() -> (Brush Cache:Slate Brush Structure))
  ]}
]
variables [
 Background Color {type=LinearColor, default=<rgba 0 0 0 0>}
 Brush Cache {
  type=SlateBrush
  default {
   bIsDynamicallyLoaded=false
   drawAs=Image
   imageType=NoImage
   mirroring=NoMirror
   resourceName=None
   resourceObject=null
   tiling=NoTile
   imageSize=<32 32>
   margin=<ltrb 0 0 0 0>
   outlineSettings {
    bUseBrushTransparency=false
    roundingType=HalfHeightRadius
    width=0
    color {colorUseRule=UseColor_Specified, specifiedColor=<rgba 0 0 0 0>}
    cornerRadii=<0 0 0 0>
   }
   tintColor {colorUseRule=UseColor_Specified, specifiedColor=<rgba 1 1 1 1>}
   uVRegion {bIsValid=false, max=<0 0>, min=<0 0>}
  }
 }
 Hide Header {type=boolean}
 Hover Outline Color {type=LinearColor, default=<rgba 0 0 0 0>}
]
widgets {
 info {inheritedWidgetCount=0, parentClass=&UserWidget, rootWidgetClass=&Border, widgetCount=4}
 widgets [
  Border {
   widgetClassPath=&Border
   bIsVariable=true
   widget=Border
   properties {
    background {
     bIsDynamicallyLoaded=false
     drawAs=RoundedBox
     imageType=NoImage
     mirroring=NoMirror
     resourceName=None
     resourceObject=null
     tiling=NoTile
     imageSize=<32 32>
     margin=<ltrb 0 0 0 0>
     outlineSettings {
      bUseBrushTransparency=false
      roundingType=FixedRadius
      width=1
      color {colorUseRule=UseColor_Specified, specifiedColor=<rgba 0 0 0 0>}
      cornerRadii=<12 12 12 12>
     }
     tintColor {colorUseRule=UseColor_Specified, specifiedColor=<rgba 1 1 1 1>}
     uVRegion {bIsValid=false, max=<0 0>, min=<0 0>}
    }
    brushColor=<rgba 0.0284 0.0284 0.0284 1>
    padding=<ltrb 8 8 8 8>
   }
  }
  VerticalBox_47 {widgetClassPath=&VerticalBox, parent=Border, widget=VerticalBox_47, slot {class=BorderSlot, properties {padding=<ltrb 8 8 8 8>}}}
  HeaderSlot {
   widgetClassPath=&NamedSlot
   bIsVariable=true
   parent=VerticalBox_47
   widget=HeaderSlot
   slot {class=VerticalBoxSlot, properties {padding=<ltrb 0 3 0 3>}}
  }
  BodySlot {
   widgetClassPath=&NamedSlot
   bIsVariable=true
   parent=VerticalBox_47
   widget=BodySlot
   slot {class=VerticalBoxSlot, properties {size {sizeRule=Fill, value=1}}}
  }
 ]
}
