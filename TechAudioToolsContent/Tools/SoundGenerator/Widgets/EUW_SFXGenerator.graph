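# Editor utility widget that serves as a custom editor for MSS_SFXGenerator MetaSound presets, for designing procedural sound effects.
# On construct it binds literal widgets to the MetaSound viewmodel through the builder and starts audio buses feeding oscilloscopes.
# Hierarchical randomizers cascade from global to section knobs, skipping locked groups.
# Auditioning watches amp envelope and LFO outputs to drive progress bars, reset on stop; a Record toggle controls WaveWriter.Enabled.
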EditorUtilityWidgetBlueprint /TechAudioToolsContent/Tools/SoundGenerator/Widgets/EUW_SFXGenerator
graphs [
 BindOutputWatch {nodes [
   n1 = BindOutputWatch!() -> (then, AudioComponentIn:Audio Component Object Reference) {is=title K2Node_FunctionEntry}
   > n2 = GetEditorWorld(self=Get UnrealEditorSubsystem())
   n3 = Get MetaSoundOutputSubsystem!(WorldContext:Object Reference=@n2) -> (ReturnValue:Meta Sound Output Subsystem Object Reference) {is=title K2Node_GetSubsystem}
   > WatchOutput!(
    execute
    self:Meta Sound Output Subsystem Object Reference=@n3
    AudioComponent:Audio Component Object Reference=@n1
    OutputName:Name=Amp.LFO.Value
    OnOutputValueChanged:"Delegate (by ref)"=Create Event()
    AnalyzerName:Name=None
    AnalyzerOutputName:Name=None
   ) -> (then, ReturnValue:bool)
   > WatchOutput(self=@n3, AudioComponent=@n1, OutputName=Amp.Value, OnOutputValueChanged=Create Event(), AnalyzerName=None, AnalyzerOutputName=None)
  ]}
 EventGraph {nodes [
   "Randomize (EUW_SFXGenerator_Randomizer_Main)"() {is=title K2Node_ComponentBoundEvent}
   > RandomizeAll()
   n1 = "On Checked State Changed (WBP_MetaSoundLiteralBoolPrimitive_ToggleButton_Record)"!() -> (OutputDelegate, then, bIsChecked:bool) {is=title K2Node_ComponentBoundEvent}
   > SetIconColor!(
    execute
    self:Tech Audio Tools Editor Slate Icon Object Reference=Get TechAudioToolsEditorSlateIcon_RecordIcon() -> (TechAudioToolsEditorSlateIcon_RecordIcon:Tech Audio Tools Editor Slate Icon Object Reference)
    InColor:Linear Color Structure=SelectColor(A=(R=1,G=0,B=0,A=1), B=(R=0.2086,G=0.2086,B=0.2086,A=1), bPickA=@n1.bIsChecked)
   ) -> (then)
   > Set!(
    execute
    ColorAndOpacity_SpecifiedColor:Linear Color Structure=SelectColor(A=(R=1,G=0,B=0,A=1), B=(R=0.0176,G=0.0176,B=0.0176,A=1), bPickA=@n1.bIsChecked)
    ColorAndOpacity_ColorUseRule:ESlateColorStylingMode Enum
    self:Text Object Reference=Get TextBlock_Armed() -> (TextBlock_Armed:Text Object Reference)
   ) -> (then, Output_Get:Slate Color Structure)
   "Randomize (EUW_SFXGenerator_Randomizer_Osc)"() {is=title K2Node_ComponentBoundEvent}
   > RandomizeOscillator()
   "Randomize (EUW_SFXGenerator_Randomizer_Generator)"() {is=title K2Node_ComponentBoundEvent}
   > RandomizeGenerator()
   "Randomize (EUW_SFXGenerator_Randomizer_PitchMod)"() {is=title K2Node_ComponentBoundEvent}
   > RandomizePitchMod()
   "Randomize (EUW_SFXGenerator_Randomizer_SpectralFX)"() {is=title K2Node_ComponentBoundEvent}
   > RandomizeSpectralFX()
   "Randomize (EUW_SFXGenerator_Randomizer_BitCrusher)"() {is=title K2Node_ComponentBoundEvent}
   > RandomizeBitCrusher()
   "Randomize (EUW_SFXGenerator_Randomizer_RingMod)"() {is=title K2Node_ComponentBoundEvent}
   > RandomizeRingMod()
   "Randomize (EUW_SFXGenerator_Randomizer_Filter)"() {is=title K2Node_ComponentBoundEvent}
   > RandomizeFilter()
   "Randomize (EUW_SFXGenerator_Randomizer_FilterControl)"() {is=title K2Node_ComponentBoundEvent}
   > RandomizeFilterControl()
   "Randomize (EUW_SFXGenerator_Randomizer_FilterMod)"() {is=title K2Node_ComponentBoundEvent}
   > RandomizeFilterModulation()
   "Randomize (EUW_SFXGenerator_Randomizer_Amp)"() {is=title K2Node_ComponentBoundEvent}
   > RandomizeAmp()
   "Randomize (EUW_SFXGenerator_Randomizer_AmpEnv)"() {is=title K2Node_ComponentBoundEvent}
   > RandomizeAmpEnvelope()
   "Randomize (EUW_SFXGenerator_Randomizer_AmpLFO)"() {is=title K2Node_ComponentBoundEvent}
   > RandomizeAmpLFO()
   "Randomize (EUW_SFXGenerator_Randomizer_SendFX)"() {is=title K2Node_ComponentBoundEvent}
   > RandomizeSendFX()
   "Randomize (EUW_SFXGenerator_Randomizer_QuadFlanger)"() {is=title K2Node_ComponentBoundEvent}
   > RandomizeSendFXQuadFlanger()
   "Randomize (EUW_SFXGenerator_Randomizer_Delay)"() {is=title K2Node_ComponentBoundEvent}
   > RandomizeSendFXDelay()
   TryInitialize() {is=title K2Node_CustomEvent}
   > n2 = Get!(execute:Exec) -> (then:Exec, else:Exec, Builder:Meta Sound Builder Base Object Reference)
   > Initialize Builder!(
    execute
    self:Meta Sound Viewmodel Object Reference=Get MetaSoundEditorViewModel() -> (MetaSoundEditorViewModel:Meta Sound Editor Viewmodel Object Reference)
    InBuilder:Meta Sound Builder Base Object Reference=@n2
   ) -> (then)
   > n3 = For Each Loop(Array:Array of Meta Sound Literal Widget Interface Interfaces=MakeMetaSoundInterfaceArray!(self) -> (OscillatorInterfaceArray:Array of Meta Sound Literal Widget Interface Interfaces)) -> (Array Element:Meta Sound Literal Widget Interface Interface)
   > Set Literal Widget Input Viewmodels!(
    execute
    Literal Widget:Meta Sound Literal Widget Interface Interface=@n3
    MetaSound Viewmodel:Meta Sound Viewmodel Object Reference=Get MetaSoundEditorViewModel()
    __WorldContext:Object Reference
   ) -> (then) {is=title K2Node_CallFunction}
   n4 = EnvOutput!() -> (OutputDelegate, then, OutputName:Name, Output:"Meta Sound Output Structure (by ref)") {is=title K2Node_CustomEvent}
   > n5 = GetFloat!(execute, Output:"Meta Sound Output Structure (by ref)"=@n4.Output) -> (then, Success:bool, ReturnValue:float)
   > n6 = Branch(Condition=@n5)
   n7 = Get EditorUtilityProgressBar_Env() -> (EditorUtilityProgressBar_Env:Editor Utility Progress Bar Object Reference)
   > SetPercent!(execute, self:Progress Bar Object Reference=@n7, InPercent:float=@n5.ReturnValue) -> (then)
   SetPercent(execute=[@n12.then, @n6.else], self=@n7, InPercent=0)
   n8 = LFOOutput!() -> (OutputDelegate, then, OutputName:Name, Output:"Meta Sound Output Structure (by ref)") {is=title K2Node_CustomEvent}
   > n9 = GetFloat(Output=@n8.Output)
   > n10 = Branch(Condition=@n9)
   > SetPercent(
    self=Get EditorUtilityProgressBar_LFO() -> (EditorUtilityProgressBar_LFO:Editor Utility Progress Bar Object Reference)
    InPercent=@n9.ReturnValue
   )
   SetPercent(execute=[@n11.then, @n10.else], self=Get EditorUtilityProgressBar_LFO(), InPercent=0)
   n11 = ResetLFO() {is=title K2Node_CustomEvent}
   n12 = ResetEnv() {is=title K2Node_CustomEvent}
   n13 = OnAudioComponentPlayState!() -> (OutputDelegate, then, PlayState:EAudioComponentPlayState Enum=Playing) {is=title K2Node_CustomEvent}
   > n14 = Switch on EAudioComponentPlayState!(execute, Selection:EAudioComponentPlayState Enum=@n13.PlayState) -> (Playing:Exec, Stopped:Exec, Paused:Exec, FadingIn:Exec, FadingOut:Exec) {is=title K2Node_SwitchEnum}
   n15 = Get!(execute=@n14.Stopped) -> (then, else, Audio Component:Audio Component Object Reference)
   > n16 = UnbindOutputWatch!(execute, self, AudioComponentIn:Audio Component Object Reference=@n15) -> (then)
   ResetLFO(execute=[@n16.then, @n15.else]) {is=title K2Node_CallFunction}
   > ResetEnv() {is=title K2Node_CallFunction}
   Event PreConstruct()
   n17 = Is Checked!(self:WBP Meta Sound Literal Bool Primitive Toggle Button SFXGen Object Reference=Get WBP_MetaSoundLiteralBoolPrimitive_ToggleButton_Record() -> (WBP_MetaSoundLiteralBoolPrimitive_ToggleButton_Record:WBP Meta Sound Literal Bool Primitive Toggle Button SFXGen Object Reference)) -> (ReturnValue:bool)
   > SetIconColor(
    self=Get TechAudioToolsEditorSlateIcon_RecordIcon()
    InColor=SelectColor(A=(R=1,G=0,B=0,A=1), B=(R=0.2086,G=0.2086,B=0.2086,A=1), bPickA=@n17)
   )
   > Set(ColorAndOpacity_SpecifiedColor=SelectColor(A=(R=1,G=0,B=0,A=1), B=(R=0.0176,G=0.0176,B=0.0176,A=1), bPickA=@n17), self=Get TextBlock_Armed())
   Event Construct()
   > TryInitialize() {is=title K2Node_CallFunction}
   > InitializeAudioBuses()
   n18 = Event On MetaSound Audition State Changed!() -> (OutputDelegate, then, AudioComponent:Audio Component Object Reference, bIsAuditioning:bool) {is=title K2Node_Event}
   > Branch(Condition=@n18.bIsAuditioning)
   > n19 = Set Audio Component(Audio Component:Audio Component Object Reference=@n18.AudioComponent) -> (Output_Get:Audio Component Object Reference)
   > BindOutputWatch!(execute, self, AudioComponentIn:Audio Component Object Reference=@n19) -> (then) {is=title K2Node_CallFunction}
   > Bind Event to On Audio Play State Changed(self:Audio Component Object Reference=Get Audio Component() -> (Audio Component:Audio Component Object Reference), Delegate=Create Event()) {is=title K2Node_AddDelegate}
   n20 = Event On MetaSoundPreset Widget Constructed!() -> (OutputDelegate, then, Builder:Meta Sound Builder Base Object Reference) {is=title K2Node_Event}
   > Set Builder(Builder:Meta Sound Builder Base Object Reference=@n20.Builder) -> (Output_Get:Meta Sound Builder Base Object Reference)
  ]}
 GetEditorName {nodes [
   GetEditorName() {is=title K2Node_FunctionEntry}
   > Return Node!(execute, ReturnValue:Text)
  ]}
 GetIconBrushName {nodes [
   GetIconBrushName() {is=title K2Node_FunctionEntry}
   > Return Node(ReturnValue:Name=None)
  ]}
 GetSupportedMetaSounds {nodes [
   GetSupportedMetaSounds() {is=title K2Node_FunctionEntry}
   > Return Node!(
    execute
    bSupportAllPresets:bool=false
    ExcludedMetaSounds:Array of Meta Sound Document Interface Interfaces
    IncludedMetaSounds:Array of Meta Sound Document Interface Interfaces=Make Array("[0]":Meta Sound Document Interface Interface=/TechAudioToolsContent/Tools/SoundGenerator/MSS_SFXGenerator) -> (Array:Array of Meta Sound Document Interface Interfaces)
   )
  ]}
 HandleLockToggleViz {nodes [
   n1 = HandleLockToggleViz!() -> (then, Locked:bool=~, LockToggleButton:Tech Audio Tools Toggle Button Icon Object Reference, RandomButton:User Widget Object Reference) {is=title K2Node_FunctionEntry}
   > n2 = Branch(Condition=@n1)
   > Set with Broadcast Unchecked Icon Brush Path!(
    execute
    Unchecked Icon Brush Path:Name=EditorStyle.Icons.lock
    self:Tech Audio Tools Toggle Button Icon Object Reference=@n1.LockToggleButton
   ) -> (then, Output_Get:Name=None)
   > Set ColorAndOpacity!(execute, ColorAndOpacity:Linear Color Structure=(R=1,G=1,B=1,A=0.2), self:User Widget Object Reference=@n1.RandomButton) -> (then, Output_Get:Linear Color Structure)
   Set with Broadcast Unchecked Icon Brush Path(execute=@n2.else, Unchecked Icon Brush Path=EditorStyle.Icons.unlock, self=@n1.LockToggleButton) -> (Output_Get=None)
   > Set ColorAndOpacity(ColorAndOpacity=(R=1,G=1,B=1,A=1), self=@n1.RandomButton)
  ]}
 InitializeAudioBuses {nodes [
   InitializeAudioBuses() {is=title K2Node_FunctionEntry}
   > n1 = For Each Loop(Array:Array of Audio Bus Object References=Get AudioBuses() -> (AudioBuses:Array of Audio Bus Object References)) -> (Array Element:Audio Bus Object Reference)
   n2 = For Each Loop(
    Exec=@n1.Completed
    Array:Array of Audio Oscilloscope Object References=Make Array!(
     "[0]":Audio Oscilloscope Object Reference=Get AudioOscilloscope_78() -> (AudioOscilloscope_78:Audio Oscilloscope Object Reference)
     "[1]":Audio Oscilloscope Object Reference=Get AudioOscilloscope_1() -> (AudioOscilloscope_1:Audio Oscilloscope Object Reference)
     "[2]":Audio Oscilloscope Object Reference=Get AudioOscilloscope() -> (AudioOscilloscope:Audio Oscilloscope Object Reference)
     "[3]":Audio Oscilloscope Object Reference=Get AudioOscilloscope_2() -> (AudioOscilloscope_2:Audio Oscilloscope Object Reference)
     "[4]":Audio Oscilloscope Object Reference=Get AudioOscilloscope_3() -> (AudioOscilloscope_3:Audio Oscilloscope Object Reference)
     "[5]":Audio Oscilloscope Object Reference=Get AudioOscilloscope_4() -> (AudioOscilloscope_4:Audio Oscilloscope Object Reference)
     "[6]":Audio Oscilloscope Object Reference=Get AudioOscilloscope_5() -> (AudioOscilloscope_5:Audio Oscilloscope Object Reference)
     "[7]":Audio Oscilloscope Object Reference=Get AudioOscilloscope_7() -> (AudioOscilloscope_7:Audio Oscilloscope Object Reference)
     "[8]":Audio Oscilloscope Object Reference=Get AudioOscilloscope_8() -> (AudioOscilloscope_8:Audio Oscilloscope Object Reference)
    ) -> (Array:Array of Audio Oscilloscope Object References)
   ) -> (Array Element:Audio Oscilloscope Object Reference)
   > StartProcessing(self:Audio Oscilloscope Object Reference=@n2)
   n3 = GetEditorWorld(execute=@n1.LoopBody, self=Get UnrealEditorSubsystem())
   > StartAudioBus!(execute, WorldContextObject:Object Reference=@n3, AudioBus:Audio Bus Object Reference=@n1) -> (then)
  ]}
 MakeMetaSoundInterfaceArray {nodes [
   MakeMetaSoundInterfaceArray() {is=title K2Node_FunctionEntry}
   > Return Node!(
    execute
    OscillatorInterfaceArray:Array of Meta Sound Literal Widget Interface Interfaces=Make Array!(
     "[0]":Meta Sound Literal Widget Interface Interface=Get EUW_SFXGeneratorKnob_RingMod_Mod_Depth() -> (EUW_SFXGeneratorKnob_RingMod_Mod_Depth:EUW Meta Sound Float Knob SFXGen Object Reference)
     "[1]":Meta Sound Literal Widget Interface Interface=Get EUW_SFXGeneratorKnob_RingMod_Control_Amount() -> (EUW_SFXGeneratorKnob_RingMod_Control_Amount:EUW Meta Sound Float Knob SFXGen Object Reference)
     "[2]":Meta Sound Literal Widget Interface Interface=Get EUW_SFXGeneratorKnob_SpectralFX_BitCrusher_Amount() -> (EUW_SFXGeneratorKnob_SpectralFX_BitCrusher_Amount:EUW Meta Sound Float Knob SFXGen Object Reference)
     "[3]":Meta Sound Literal Widget Interface Interface=Get EUW_SFXGeneratorKnob_RingMod_Mod_Curve() -> (EUW_SFXGeneratorKnob_RingMod_Mod_Curve:EUW Meta Sound Float Knob SFXGen Object Reference)
     "[4]":Meta Sound Literal Widget Interface Interface=Get EUW_SFXGeneratorKnob_RingMod_Mod_Shape() -> (EUW_SFXGeneratorKnob_RingMod_Mod_Shape:EUW Meta Sound Float Knob SFXGen Object Reference)
     "[5]":Meta Sound Literal Widget Interface Interface=Get EUW_SFXGeneratorKnob_RingMod_Control_Frequency() -> (EUW_SFXGeneratorKnob_RingMod_Control_Frequency:EUW Meta Sound Float Knob SFXGen Object Reference)
     "[6]":Meta Sound Literal Widget Interface Interface=Get EUW_SFXGeneratorKnob_SpectralFX_BitCrusher_SampleRate() -> (EUW_SFXGeneratorKnob_SpectralFX_BitCrusher_SampleRate:EUW Meta Sound Float Knob SFXGen Object Reference)
     "[7]":Meta Sound Literal Widget Interface Interface=Get EUW_SFXGeneratorKnob_SpectralFX_WaveShaper() -> (EUW_SFXGeneratorKnob_SpectralFX_WaveShaper:EUW Meta Sound Float Knob SFXGen Object Reference)
     "[8]":Meta Sound Literal Widget Interface Interface=Get EUW_SFXGeneratorKnob_Osc_PW_DutyCycle() -> (EUW_SFXGeneratorKnob_Osc_PW_DutyCycle:EUW Meta Sound Float Knob SFXGen Object Reference)
     "[9]":Meta Sound Literal Widget Interface Interface=Get EUW_SFXGeneratorKnob_Osc_PW_Sweep() -> (EUW_SFXGeneratorKnob_Osc_PW_Sweep:EUW Meta Sound Float Knob SFXGen Object Reference)
     "[10]":Meta Sound Literal Widget Interface Interface=Get EUW_SFXGeneratorKnob_Osc_Frequency() -> (EUW_SFXGeneratorKnob_Osc_Frequency:EUW Meta Sound Float Knob SFXGen Object Reference)
     "[11]":Meta Sound Literal Widget Interface Interface=Get EUW_SFXGeneratorKnob_Osc_Waveform() -> (EUW_SFXGeneratorKnob_Osc_Waveform:EUW Meta Sound Float Knob SFXGen Object Reference)
     "[12]":Meta Sound Literal Widget Interface Interface=Get EUW_SFXGeneratorKnob_PM_Vibrato_Freq() -> (EUW_SFXGeneratorKnob_PM_Vibrato_Freq:EUW Meta Sound Float Knob SFXGen Object Reference)
     "[13]":Meta Sound Literal Widget Interface Interface=Get EUW_SFXGeneratorKnob_PM_Vibrato_Depth() -> (EUW_SFXGeneratorKnob_PM_Vibrato_Depth:EUW Meta Sound Float Knob SFXGen Object Reference)
     "[14]":Meta Sound Literal Widget Interface Interface=Get EUW_SFXGeneratorKnob_PM_Sweep_Dir() -> (EUW_SFXGeneratorKnob_PM_Sweep_Dir:EUW Meta Sound Float Knob SFXGen Object Reference)
     "[15]":Meta Sound Literal Widget Interface Interface=Get EUW_SFXGeneratorKnob_PM_Sweep_Curve() -> (EUW_SFXGeneratorKnob_PM_Sweep_Curve:EUW Meta Sound Float Knob SFXGen Object Reference)
     "[16]":Meta Sound Literal Widget Interface Interface=Get EUW_MetaSoundInteger_Slider_Vertical_SFXGen_PM_Jump_Num() -> (EUW_MetaSoundInteger_Slider_Vertical_SFXGen_PM_Jump_Num:EUW Meta Sound Integer Slider Vertical SFXGen Object Reference)
     "[17]":Meta Sound Literal Widget Interface Interface=Get EUW_SFXGeneratorKnob_PM_Jump_Speed() -> (EUW_SFXGeneratorKnob_PM_Jump_Speed:EUW Meta Sound Float Knob SFXGen Object Reference)
     "[18]":Meta Sound Literal Widget Interface Interface=Get EUW_SFXGeneratorKnob_PM_Jump_Size() -> (EUW_SFXGeneratorKnob_PM_Jump_Size:EUW Meta Sound Float Knob SFXGen Object Reference)
     "[19]":Meta Sound Literal Widget Interface Interface=Get EUW_SFXGeneratorKnob_Filter_Control_Cutoff() -> (EUW_SFXGeneratorKnob_Filter_Control_Cutoff:EUW Meta Sound Float Knob SFXGen Object Reference)
     "[20]":Meta Sound Literal Widget Interface Interface=Get EUW_SFXGeneratorKnob_AmpEnv_Decay() -> (EUW_SFXGeneratorKnob_AmpEnv_Decay:EUW Meta Sound Float Knob SFXGen Object Reference)
     "[21]":Meta Sound Literal Widget Interface Interface=Get EUW_SFXGeneratorKnob_Filter_Mod_Curve() -> (EUW_SFXGeneratorKnob_Filter_Mod_Curve:EUW Meta Sound Float Knob SFXGen Object Reference)
     "[22]":Meta Sound Literal Widget Interface Interface=Get EUW_SFXGeneratorKnob_AmpEnv_Sustain() -> (EUW_SFXGeneratorKnob_AmpEnv_Sustain:EUW Meta Sound Float Knob SFXGen Object Reference)
     "[23]":Meta Sound Literal Widget Interface Interface=Get EUW_SFXGeneratorKnob_Filter_Control_Resonance() -> (EUW_SFXGeneratorKnob_Filter_Control_Resonance:EUW Meta Sound Float Knob SFXGen Object Reference)
     "[24]":Meta Sound Literal Widget Interface Interface=Get EUW_SFXGeneratorKnob_AmpLFO_Freq() -> (EUW_SFXGeneratorKnob_AmpLFO_Freq:EUW Meta Sound Float Knob SFXGen Object Reference)
     "[25]":Meta Sound Literal Widget Interface Interface=Get EUW_SFXGeneratorKnob_AmpEnv_Duration() -> (EUW_SFXGeneratorKnob_AmpEnv_Duration:EUW Meta Sound Float Knob SFXGen Object Reference)
     "[26]":Meta Sound Literal Widget Interface Interface=Get EUW_SFXGeneratorKnob_Filter_Mod_Depth() -> (EUW_SFXGeneratorKnob_Filter_Mod_Depth:EUW Meta Sound Float Knob SFXGen Object Reference)
     "[27]":Meta Sound Literal Widget Interface Interface=Get EUW_SFXGeneratorKnob_AmpLFO_Depth() -> (EUW_SFXGeneratorKnob_AmpLFO_Depth:EUW Meta Sound Float Knob SFXGen Object Reference)
     "[28]":Meta Sound Literal Widget Interface Interface=Get EUW_SFXGeneratorKnob_Filter_Mod_Shape() -> (EUW_SFXGeneratorKnob_Filter_Mod_Shape:EUW Meta Sound Float Knob SFXGen Object Reference)
     "[29]":Meta Sound Literal Widget Interface Interface=Get EUW_SFXGeneratorKnob_AmpEnv_Release() -> (EUW_SFXGeneratorKnob_AmpEnv_Release:EUW Meta Sound Float Knob SFXGen Object Reference)
     "[30]":Meta Sound Literal Widget Interface Interface=Get EUW_SFXGeneratorKnob_AmpEnv_Attack() -> (EUW_SFXGeneratorKnob_AmpEnv_Attack:EUW Meta Sound Float Knob SFXGen Object Reference)
     "[31]":Meta Sound Literal Widget Interface Interface=Get EUW_SFXGeneratorKnob_Filter_Control_Type() -> (EUW_SFXGeneratorKnob_Filter_Control_Type:EUW Meta Sound Float Knob SFXGen Object Reference)
     "[32]":Meta Sound Literal Widget Interface Interface=Get EUW_SFXGeneratorKnob_AmpEnv_Decay()
     "[33]":Meta Sound Literal Widget Interface Interface=Get EUW_SFXGeneratorKnob_AmpEnv_Sustain()
     "[34]":Meta Sound Literal Widget Interface Interface=Get EUW_SFXGeneratorKnob_SendFX_Delay_Amount() -> (EUW_SFXGeneratorKnob_SendFX_Delay_Amount:EUW Meta Sound Float Knob SFXGen Object Reference)
     "[35]":Meta Sound Literal Widget Interface Interface=Get EUW_SFXGeneratorKnob_AmpEnv_Duration()
     "[36]":Meta Sound Literal Widget Interface Interface=Get EUW_SFXGeneratorKnob_SendFX_Delay_Time() -> (EUW_SFXGeneratorKnob_SendFX_Delay_Time:EUW Meta Sound Float Knob SFXGen Object Reference)
     "[37]":Meta Sound Literal Widget Interface Interface=Get EUW_SFXGeneratorKnob_AmpEnv_Release()
     "[38]":Meta Sound Literal Widget Interface Interface=Get EUW_SFXGeneratorKnob_SendFX_Delay_Feedback() -> (EUW_SFXGeneratorKnob_SendFX_Delay_Feedback:EUW Meta Sound Float Knob SFXGen Object Reference)
     "[39]":Meta Sound Literal Widget Interface Interface=Get EUW_SFXGeneratorKnob_SendFX_QuadFlanger_Amount() -> (EUW_SFXGeneratorKnob_SendFX_QuadFlanger_Amount:EUW Meta Sound Float Knob SFXGen Object Reference)
     "[40]":Meta Sound Literal Widget Interface Interface=Get EUW_SFXGeneratorKnob_SendFX_Reverb_Amount() -> (EUW_SFXGeneratorKnob_SendFX_Reverb_Amount:EUW Meta Sound Float Knob SFXGen Object Reference)
     "[41]":Meta Sound Literal Widget Interface Interface=Get EUW_SFXGeneratorKnob_SendFX_QuadFlanger_DelayTime() -> (EUW_SFXGeneratorKnob_SendFX_QuadFlanger_DelayTime:EUW Meta Sound Float Knob SFXGen Object Reference)
     "[42]":Meta Sound Literal Widget Interface Interface=Get WBP_MetaSoundLiteralStringPrimitive_EditableText_SFXGen() -> (WBP_MetaSoundLiteralStringPrimitive_EditableText_SFXGen:WBP Meta Sound Literal String Primitive Editable Text SFXGen Object Reference)
     "[43]":Meta Sound Literal Widget Interface Interface=Get WBP_MetaSoundLiteralBoolPrimitive_ToggleButton_Record()
    ) -> (Array:Array of Meta Sound Literal Widget Interface Interfaces)
   )
  ]}
 OnInitializingTab {nodes [
   On Initializing Tab() {is=title K2Node_FunctionEntry}
   > Return Node!(execute, TabInfo:Meta Sound Template Widget Tab Info Structure)
  ]}
 RandomizeAll {nodes [
   RandomizeAll() {is=title K2Node_FunctionEntry}
   n1 = Get EUW_SFXGenerator_Randomizer_Main() -> (EUW_SFXGenerator_Randomizer_Main:EUW SFXGenerator Randomizer Object Reference)
   > Branch(Condition=NOT Boolean(A=GetIsLocked!(self:EUW SFXGenerator Randomizer Object Reference=@n1) -> (bIsChecked:bool)))
   > RandomizeGenerator()
   > RandomizeSpectralFX()
   > RandomizeFilter()
   > RandomizeAmp()
   > RandomizeSendFX()
  ]}
 RandomizeAmp {nodes [
   RandomizeAmp() {is=title K2Node_FunctionEntry}
   n1 = Get EUW_SFXGenerator_Randomizer_Amp() -> (EUW_SFXGenerator_Randomizer_Amp:EUW SFXGenerator Randomizer Object Reference)
   > Branch(Condition=NOT Boolean(A=GetIsLocked(self=@n1)))
   > RandomizeAmpLFO()
   > RandomizeAmpEnvelope()
  ]}
 RandomizeAmpEnvelope {nodes [
   RandomizeAmpEnvelope() {is=title K2Node_FunctionEntry}
   n1 = Get EUW_SFXGenerator_Randomizer_AmpEnv() -> (EUW_SFXGenerator_Randomizer_AmpEnv:EUW SFXGenerator Randomizer Object Reference)
   > Branch(Condition=NOT Boolean(A=GetIsLocked(self=@n1)))
   > n2 = For Each Loop(Array:Array of EUW Meta Sound Float Knob SFXGen Object References=Make Array!(
     "[0]":EUW Meta Sound Float Knob SFXGen Object Reference=Get EUW_SFXGeneratorKnob_AmpEnv_Attack()
     "[1]":EUW Meta Sound Float Knob SFXGen Object Reference=Get EUW_SFXGeneratorKnob_AmpEnv_Decay()
     "[2]":EUW Meta Sound Float Knob SFXGen Object Reference=Get EUW_SFXGeneratorKnob_AmpEnv_Duration()
     "[3]":EUW Meta Sound Float Knob SFXGen Object Reference=Get EUW_SFXGeneratorKnob_AmpEnv_Release()
     "[4]":EUW Meta Sound Float Knob SFXGen Object Reference=Get EUW_SFXGeneratorKnob_AmpEnv_Sustain()
    ) -> (Array:Array of EUW Meta Sound Float Knob SFXGen Object References)) -> (Array Element:EUW Meta Sound Float Knob SFXGen Object Reference)
   > RandomizeValue(self:EUW Meta Sound Float Knob SFXGen Object Reference=@n2)
  ]}
 RandomizeAmpLFO {nodes [
   RandomizeAmpLFO() {is=title K2Node_FunctionEntry}
   n1 = Get EUW_SFXGenerator_Randomizer_AmpLFO() -> (EUW_SFXGenerator_Randomizer_AmpLFO:EUW SFXGenerator Randomizer Object Reference)
   > Branch(Condition=NOT Boolean(A=GetIsLocked(self=@n1)))
   > n2 = For Each Loop(Array=Make Array!("[0]"=Get EUW_SFXGeneratorKnob_AmpLFO_Freq(), "[1]"=Get EUW_SFXGeneratorKnob_AmpLFO_Depth()) -> (Array))
   > RandomizeValue(self=@n2)
  ]}
 RandomizeBitCrusher {nodes [
   RandomizeBitCrusher() {is=title K2Node_FunctionEntry}
   n1 = Get EUW_SFXGenerator_Randomizer_BitCrusher() -> (EUW_SFXGenerator_Randomizer_BitCrusher:EUW SFXGenerator Randomizer Object Reference)
   > Branch(Condition=NOT Boolean(A=GetIsLocked(self=@n1)))
   > RandomizeValue(self=Get EUW_SFXGeneratorKnob_SpectralFX_BitCrusher_Amount())
   > RandomizeValue(self=Get EUW_SFXGeneratorKnob_SpectralFX_BitCrusher_SampleRate())
  ]}
 RandomizeFilter {nodes [
   RandomizeFilter() {is=title K2Node_FunctionEntry}
   n1 = Get EUW_SFXGenerator_Randomizer_Filter() -> (EUW_SFXGenerator_Randomizer_Filter:EUW SFXGenerator Randomizer Object Reference)
   > Branch(Condition=NOT Boolean(A=GetIsLocked(self=@n1)))
   > RandomizeFilterControl()
   > RandomizeFilterModulation()
  ]}
 RandomizeFilterControl {nodes [
   RandomizeFilterControl() {is=title K2Node_FunctionEntry}
   n1 = Get EUW_SFXGenerator_Randomizer_FilterControl() -> (EUW_SFXGenerator_Randomizer_FilterControl:EUW SFXGenerator Randomizer Object Reference)
   > Branch(Condition=NOT Boolean(A=GetIsLocked(self=@n1)))
   > n2 = For Each Loop(Array=Make Array!(
     "[0]"=Get EUW_SFXGeneratorKnob_Filter_Control_Type()
     "[1]"=Get EUW_SFXGeneratorKnob_Filter_Control_Cutoff()
     "[2]":EUW Meta Sound Float Knob SFXGen Object Reference=Get EUW_SFXGeneratorKnob_Filter_Control_Resonance()
    ) -> (Array))
   > RandomizeValue(self=@n2)
  ]}
 RandomizeFilterModulation {nodes [
   RandomizeFilterModulation() {is=title K2Node_FunctionEntry}
   n1 = Get EUW_SFXGenerator_Randomizer_FilterMod() -> (EUW_SFXGenerator_Randomizer_FilterMod:EUW SFXGenerator Randomizer Object Reference)
   > Branch(Condition=NOT Boolean(A=GetIsLocked(self=@n1)))
   > n2 = For Each Loop(Array=Make Array(
     "[0]"=Get EUW_SFXGeneratorKnob_Filter_Mod_Depth()
     "[1]"=Get EUW_SFXGeneratorKnob_Filter_Mod_Shape()
     "[2]"=Get EUW_SFXGeneratorKnob_Filter_Mod_Curve()
    ))
   > RandomizeValue(self=@n2)
  ]}
 RandomizeGenerator {nodes [
   RandomizeGenerator() {is=title K2Node_FunctionEntry}
   n1 = Get EUW_SFXGenerator_Randomizer_Generator() -> (EUW_SFXGenerator_Randomizer_Generator:EUW SFXGenerator Randomizer Object Reference)
   > Branch(Condition=NOT Boolean(A=GetIsLocked(self=@n1)))
   > RandomizeOscillator()
   > RandomizePitchMod()
  ]}
 RandomizeOscillator {nodes [
   RandomizeOscillator() {is=title K2Node_FunctionEntry}
   n1 = Get EUW_SFXGenerator_Randomizer_Osc() -> (EUW_SFXGenerator_Randomizer_Osc:EUW SFXGenerator Randomizer Object Reference)
   > Branch(Condition=NOT Boolean(A=GetIsLocked(self=@n1)))
   > n2 = For Each Loop(Array=Make Array!(
     "[0]"=Get EUW_SFXGeneratorKnob_Osc_Frequency()
     "[1]"=Get EUW_SFXGeneratorKnob_Osc_Waveform()
     "[2]"=Get EUW_SFXGeneratorKnob_Osc_PW_Sweep()
     "[3]":EUW Meta Sound Float Knob SFXGen Object Reference=Get EUW_SFXGeneratorKnob_Osc_PW_DutyCycle()
    ) -> (Array))
   > RandomizeValue(self=@n2)
  ]}
 RandomizePitchMod {nodes [
   RandomizePitchMod() {is=title K2Node_FunctionEntry}
   n1 = Get EUW_SFXGenerator_Randomizer_PitchMod() -> (EUW_SFXGenerator_Randomizer_PitchMod:EUW SFXGenerator Randomizer Object Reference)
   > Branch(Condition=NOT Boolean(A=GetIsLocked(self=@n1)))
   > n2 = For Each Loop(Array=Make Array!(
     "[0]"=Get EUW_SFXGeneratorKnob_PM_Vibrato_Freq()
     "[1]"=Get EUW_SFXGeneratorKnob_PM_Vibrato_Depth()
     "[2]"=Get EUW_SFXGeneratorKnob_PM_Sweep_Dir()
     "[3]"=Get EUW_SFXGeneratorKnob_PM_Sweep_Curve()
     "[4]":EUW Meta Sound Float Knob SFXGen Object Reference=Get EUW_SFXGeneratorKnob_PM_Jump_Speed()
     "[5]":EUW Meta Sound Float Knob SFXGen Object Reference=Get EUW_SFXGeneratorKnob_PM_Jump_Size()
    ) -> (Array))
   RandomizeValue(
    execute=@n2.Completed
    self:EUW Meta Sound Integer Slider Vertical SFXGen Object Reference=Get EUW_MetaSoundInteger_Slider_Vertical_SFXGen_PM_Jump_Num()
   )
   RandomizeValue(execute=@n2.LoopBody, self:EUW Meta Sound Float Knob SFXGen Object Reference=@n2)
  ]}
 RandomizeRingMod {nodes [
   RandomizeRingMod() {is=title K2Node_FunctionEntry}
   n1 = Get EUW_SFXGenerator_Randomizer_RingMod() -> (EUW_SFXGenerator_Randomizer_RingMod:EUW SFXGenerator Randomizer Object Reference)
   > Branch(Condition=NOT Boolean(A=GetIsLocked(self=@n1)))
   > n2 = For Each Loop(Array=Make Array!(
     "[0]"=Get EUW_SFXGeneratorKnob_RingMod_Control_Amount()
     "[1]"=Get EUW_SFXGeneratorKnob_RingMod_Control_Frequency()
     "[2]"=Get EUW_SFXGeneratorKnob_RingMod_Mod_Depth()
     "[3]"=Get EUW_SFXGeneratorKnob_RingMod_Mod_Curve()
     "[4]"=Get EUW_SFXGeneratorKnob_RingMod_Mod_Shape()
    ) -> (Array))
   > RandomizeValue(self=@n2)
  ]}
 RandomizeSendFX {nodes [
   RandomizeSendFX() {is=title K2Node_FunctionEntry}
   n1 = Get EUW_SFXGenerator_Randomizer_SendFX() -> (EUW_SFXGenerator_Randomizer_SendFX:EUW SFXGenerator Randomizer Object Reference)
   > Branch(Condition=NOT Boolean(A=GetIsLocked(self=@n1)))
   > RandomizeSendFXQuadFlanger()
   > RandomizeSendFXDelay()
   > RandomizeValue(self=Get EUW_SFXGeneratorKnob_SendFX_Reverb_Amount())
  ]}
 RandomizeSendFXDelay {nodes [
   RandomizeSendFXDelay() {is=title K2Node_FunctionEntry}
   n1 = Get EUW_SFXGenerator_Randomizer_Delay() -> (EUW_SFXGenerator_Randomizer_Delay:EUW SFXGenerator Randomizer Object Reference)
   > Branch(Condition=NOT Boolean(A=GetIsLocked(self=@n1)))
   > n2 = For Each Loop(Array=Make Array!(
     "[0]"=Get EUW_SFXGeneratorKnob_SendFX_Delay_Amount()
     "[1]"=Get EUW_SFXGeneratorKnob_SendFX_Delay_Time()
     "[2]"=Get EUW_SFXGeneratorKnob_SendFX_Delay_Feedback()
    ) -> (Array))
   > RandomizeValue(self=@n2)
  ]}
 RandomizeSendFXQuadFlanger {nodes [
   RandomizeSendFXQuadFlanger() {is=title K2Node_FunctionEntry}
   n1 = Get EUW_SFXGenerator_Randomizer_QuadFlanger() -> (EUW_SFXGenerator_Randomizer_QuadFlanger:EUW SFXGenerator Randomizer Object Reference)
   > Branch(Condition=NOT Boolean(A=GetIsLocked(self=@n1)))
   > n2 = For Each Loop(Array=Make Array!("[0]"=Get EUW_SFXGeneratorKnob_SendFX_QuadFlanger_Amount(), "[1]"=Get EUW_SFXGeneratorKnob_SendFX_QuadFlanger_DelayTime()) -> (Array))
   > RandomizeValue(self=@n2)
  ]}
 RandomizeSpectralFX {nodes [
   RandomizeSpectralFX() {is=title K2Node_FunctionEntry}
   n1 = Get EUW_SFXGenerator_Randomizer_SpectralFX() -> (EUW_SFXGenerator_Randomizer_SpectralFX:EUW SFXGenerator Randomizer Object Reference)
   > Branch(Condition=NOT Boolean(A=GetIsLocked(self=@n1)))
   > RandomizeBitCrusher() {is=title K2Node_CallFunction}
   > RandomizeRingMod() {is=title K2Node_CallFunction}
   > RandomizeValue(self=Get EUW_SFXGeneratorKnob_SpectralFX_WaveShaper())
  ]}
 UnbindOutputWatch {nodes [
   n1 = UnbindOutputWatch!() -> (then, AudioComponentIn:Audio Component Object Reference) {is=title K2Node_FunctionEntry}
   > n2 = GetEditorWorld(self=Get UnrealEditorSubsystem())
   n3 = Get MetaSoundOutputSubsystem(WorldContext=@n2)
   > UnwatchOutput!(
    execute
    self:Meta Sound Output Subsystem Object Reference=@n3
    AudioComponent:Audio Component Object Reference=@n1
    OutputName:Name=Amp.LFO.Value
    OnOutputValueChanged:"Delegate (by ref)"=Create Event()
    AnalyzerName:Name=None
    AnalyzerOutputName:Name=None
   ) -> (then, ReturnValue:bool)
   > UnwatchOutput(self=@n3, AudioComponent=@n1, OutputName=Amp.Value, OnOutputValueChanged=Create Event(), AnalyzerName=None, AnalyzerOutputName=None)
  ]}
]
variables [
 Audio Component {type=AudioComponent, default=null}
 AudioBuses {
  type="array<AudioBus>"
  default [
   /TechAudioToolsContent/Tools/SoundGenerator/AudioBuses/AB_SFXG_BitCrusher
   /TechAudioToolsContent/Tools/SoundGenerator/AudioBuses/AB_SFXG_Delay
   /TechAudioToolsContent/Tools/SoundGenerator/AudioBuses/AB_SFXG_Filter
   /TechAudioToolsContent/Tools/SoundGenerator/AudioBuses/AB_SFXG_Osc
   /TechAudioToolsContent/Tools/SoundGenerator/AudioBuses/AB_SFXG_PitchBend
   /TechAudioToolsContent/Tools/SoundGenerator/AudioBuses/AB_SFXG_Reverb
   /TechAudioToolsContent/Tools/SoundGenerator/AudioBuses/AB_SFXG_RingMod
   /TechAudioToolsContent/Tools/SoundGenerator/AudioBuses/AB_SFXG_Vibrato
   /TechAudioToolsContent/Tools/SoundGenerator/AudioBuses/AB_SFXG_WaveShaper
  ]
 }
 Builder {type=MetaSoundBuilderBase, default=null}
]
widgets {
 info {namedSlotCount=87, rootWidgetClass=&Border, widgetCount=282}
 widgets [
  Border_0 {
   widgetClassPath=&Border
   widget=Border_0
   properties {horizontalAlignment=HAlign_Center, verticalAlignment=VAlign_Center, brushColor=<rgba 0.0103 0.0103 0.0103 1>}
  }
  HorizontalBox_179 {
   widgetClassPath=&HorizontalBox
   parent=Border_0
   widget=HorizontalBox_179
   slot {class=BorderSlot, properties {horizontalAlignment=HAlign_Center, verticalAlignment=VAlign_Center}}
  }
  WBP_SFXGeneratorContentPanel_105 {
   widgetClassPath=&/WBP_SFXGeneratorContentPanel
   bIsVariable=true
   parent=HorizontalBox_179
   widget=WBP_SFXGeneratorContentPanel_105
   slot {
    class=HorizontalBoxSlot
    properties {horizontalAlignment=HAlign_Center, verticalAlignment=VAlign_Center, padding=<ltrb 20 20 20 20>, size {sizeRule=Fill, value=1}}
   }
  }
  HorizontalBox_180 {widgetClassPath=&HorizontalBox, namedSlotHost=WBP_SFXGeneratorContentPanel_105, widget=HorizontalBox_180}
  TextBlock_332 {
   widgetClassPath=&TextBlock
   parent=HorizontalBox_180
   widget=TextBlock_332
   properties {
    text=SFX Generator
    font {
     bForceMonospaced=false
     bMaterialIsStencil=false
     fontMaterial=null
     fontObject=&Roboto
     letterSpacing=0
     monospacedWidth=1
     size=32
     skewAmount=0
     typefaceFontName=Bold
     outlineSettings {
      bApplyOutlineToDropShadows=false
      bMiteredCorners=false
      bSeparateFillAlpha=false
      outlineMaterial=null
      outlineSize=0
      outlineColor=<rgba 0 0 0 1>
     }
    }
   }
   slot {class=HorizontalBoxSlot, properties {padding=<ltrb 10 10 10 10>}}
  }
  Spacer_268 {widgetClassPath=&Spacer, parent=HorizontalBox_180, widget=Spacer_268, properties {size=<100 1>}, slot {class=HorizontalBoxSlot}}
  Border_48 {
   widgetClassPath=&Border
   parent=HorizontalBox_180
   widget=Border_48
   properties {brushColor=<rgba 0.0176 0.0176 0.0176 1>}
   slot {class=HorizontalBoxSlot, properties {padding=<ltrb 10 10 10 10>}}
  }
  EUW_SFXGenerator_Randomizer_Main {
   widgetClassPath=&/EUW_SFXGenerator_Randomizer
   bIsVariable=true
   parent=Border_48
   widget=EUW_SFXGenerator_Randomizer_Main
   properties {iconSize=<32 32>}
   slot {class=BorderSlot}
  }
  Border_64 {
   widgetClassPath=&Border
   parent=HorizontalBox_180
   widget=Border_64
   properties {brushColor=<rgba 0.0176 0.0176 0.0176 1>}
   slot {class=HorizontalBoxSlot, properties {padding=<ltrb 10 10 10 10>}}
  }
  HorizontalBox_20 {widgetClassPath=&HorizontalBox, parent=Border_64, widget=HorizontalBox_20, slot {class=BorderSlot}}
  TextBlock_Armed {
   widgetClassPath=&TextBlock
   bIsVariable=true
   parent=HorizontalBox_20
   widget=TextBlock_Armed
   properties {text=Record, colorAndOpacity {colorUseRule=UseColor_Specified, specifiedColor=<rgba 0.0176 0.0176 0.0176 1>}}
   slot {class=HorizontalBoxSlot, properties {verticalAlignment=VAlign_Center}}
  }
  SizeBox_11 {
   widgetClassPath=&SizeBox
   parent=HorizontalBox_20
   widget=SizeBox_11
   properties {bOverride_MaxAspectRatio=true, bOverride_MinAspectRatio=true, heightOverride=64, minDesiredHeight=64, minDesiredWidth=64, widthOverride=64}
   slot {class=HorizontalBoxSlot}
  }
  WBP_MetaSoundLiteralBoolPrimitive_ToggleButton_Record {
   widgetClassPath=&/WBP_MetaSoundLiteralBoolPrimitive_ToggleButton_SFXGen
   bIsVariable=true
   parent=SizeBox_11
   widget=WBP_MetaSoundLiteralBoolPrimitive_ToggleButton_Record
   properties {
    input Name=WaveWriter.Enabled
    check Box Style {
     checkBoxType=ToggleButton
     backgroundHoveredImage {
      bIsDynamicallyLoaded=false
      drawAs=RoundedBox
      imageType=NoImage
      mirroring=NoMirror
      resourceName=None
      resourceObject=null
      tiling=NoTile
      imageSize=<18 18>
      margin=<ltrb 0 0 0 0>
      outlineSettings {
       bUseBrushTransparency=false
       roundingType=FixedRadius
       width=1
       color {colorUseRule=UseColor_ColorTable, specifiedColor=<rgba 0 0 0 0>}
       cornerRadii=<3 3 3 3>
      }
      tintColor {colorUseRule=UseColor_Specified, specifiedColor=<rgba 0.0048 0.0048 0.0048 1>}
      uVRegion {bIsValid=false, max=<0 0>, min=<0 0>}
     }
     backgroundImage {
      bIsDynamicallyLoaded=false
      drawAs=RoundedBox
      imageType=NoImage
      mirroring=NoMirror
      resourceName=None
      resourceObject=null
      tiling=NoTile
      imageSize=<18 18>
      margin=<ltrb 0 0 0 0>
      outlineSettings {
       bUseBrushTransparency=false
       roundingType=FixedRadius
       width=1
       color {colorUseRule=UseColor_ColorTable, specifiedColor=<rgba 0 0 0 0>}
       cornerRadii=<3 3 3 3>
      }
      tintColor {colorUseRule=UseColor_Specified, specifiedColor=<rgba 0.0048 0.0048 0.0048 1>}
      uVRegion {bIsValid=false, max=<0 0>, min=<0 0>}
     }
     backgroundPressedImage {
      bIsDynamicallyLoaded=false
      drawAs=RoundedBox
      imageType=NoImage
      mirroring=NoMirror
      resourceName=None
      resourceObject=null
      tiling=NoTile
      imageSize=<18 18>
      margin=<ltrb 0 0 0 0>
      outlineSettings {
       bUseBrushTransparency=false
       roundingType=FixedRadius
       width=1
       color {colorUseRule=UseColor_ColorTable, specifiedColor=<rgba 0 0 0 0>}
       cornerRadii=<3 3 3 3>
      }
      tintColor {colorUseRule=UseColor_Specified, specifiedColor=<rgba 0.0048 0.0048 0.0048 1>}
      uVRegion {bIsValid=false, max=<0 0>, min=<0 0>}
     }
     borderBackgroundColor {colorUseRule=UseColor_Specified, specifiedColor=<rgba 1 1 1 1>}
     checkedForeground {colorUseRule=UseColor_Specified, specifiedColor=<rgba 1 1 1 1>}
     checkedHoveredForeground {colorUseRule=UseColor_Specified, specifiedColor=<rgba 1 1 1 1>}
     checkedHoveredImage {
      bIsDynamicallyLoaded=false
      drawAs=RoundedBox
      imageType=NoImage
      mirroring=NoMirror
      resourceName=None
      resourceObject=null
      tiling=NoTile
      imageSize=<0 0>
      margin=<ltrb 0 0 0 0>
      outlineSettings {
       bUseBrushTransparency=false
       roundingType=FixedRadius
       width=0
       color {colorUseRule=UseColor_Foreground, specifiedColor=<rgba 0 0 0 0>}
       cornerRadii=<4 4 4 4>
      }
      tintColor {colorUseRule=UseColor_Specified, specifiedColor=<rgba 0.0176 0.0176 0.0176 1>}
      uVRegion {bIsValid=false, max=<0 0>, min=<0 0>}
     }
     checkedImage {
      bIsDynamicallyLoaded=false
      drawAs=RoundedBox
      imageType=NoImage
      mirroring=NoMirror
      resourceName=None
      resourceObject=null
      tiling=NoTile
      imageSize=<0 0>
      margin=<ltrb 0 0 0 0>
      outlineSettings {
       bUseBrushTransparency=false
       roundingType=FixedRadius
       width=0
       color {colorUseRule=UseColor_Foreground, specifiedColor=<rgba 0 0 0 0>}
       cornerRadii=<4 4 4 4>
      }
      tintColor {colorUseRule=UseColor_Specified, specifiedColor=<rgba 0.0103 0.0103 0.0103 1>}
      uVRegion {bIsValid=false, max=<0 0>, min=<0 0>}
     }
     checkedPressedForeground {colorUseRule=UseColor_Specified, specifiedColor=<rgba 1 1 1 1>}
     checkedPressedImage {
      bIsDynamicallyLoaded=false
      drawAs=RoundedBox
      imageType=NoImage
      mirroring=NoMirror
      resourceName=None
      resourceObject=null
      tiling=NoTile
      imageSize=<0 0>
      margin=<ltrb 0 0 0 0>
      outlineSettings {
       bUseBrushTransparency=false
       roundingType=FixedRadius
       width=0
       color {colorUseRule=UseColor_Foreground, specifiedColor=<rgba 0 0 0 0>}
       cornerRadii=<4 4 4 4>
      }
      tintColor {colorUseRule=UseColor_Specified, specifiedColor=<rgba 0.0625 0.0625 0.0625 1>}
      uVRegion {bIsValid=false, max=<0 0>, min=<0 0>}
     }
     checkedSlateSound {resourceObject=null}
     foregroundColor {colorUseRule=UseColor_Specified, specifiedColor=<rgba 1 1 1 1>}
     hoveredForeground {colorUseRule=UseColor_Specified, specifiedColor=<rgba 1 1 1 1>}
     hoveredSlateSound {resourceObject=null}
     padding=<ltrb 8 5 8 5>
     pressedForeground {colorUseRule=UseColor_Specified, specifiedColor=<rgba 1 1 1 1>}
     uncheckedHoveredImage {
      bIsDynamicallyLoaded=false
      drawAs=RoundedBox
      imageType=NoImage
      mirroring=NoMirror
      resourceName=None
      resourceObject=null
      tiling=NoTile
      imageSize=<0 0>
      margin=<ltrb 0 0 0 0>
      outlineSettings {
       bUseBrushTransparency=false
       roundingType=FixedRadius
       width=0
       color {colorUseRule=UseColor_Specified, specifiedColor=<rgba 0.0395 0.0395 0.0395 1>}
       cornerRadii=<4 4 4 4>
      }
      tintColor {colorUseRule=UseColor_Specified, specifiedColor=<rgba 0.0395 0.0395 0.0395 1>}
      uVRegion {bIsValid=false, max=<0 0>, min=<0 0>}
     }
     uncheckedImage {
      bIsDynamicallyLoaded=false
      drawAs=NoDrawType
      imageType=NoImage
      mirroring=NoMirror
      resourceName=None
      resourceObject=null
      tiling=NoTile
      imageSize=<0 0>
      margin=<ltrb 0 0 0 0>
      outlineSettings {
       bUseBrushTransparency=false
       roundingType=FixedRadius
       width=0
       color {colorUseRule=UseColor_Foreground, specifiedColor=<rgba 0 0 0 0>}
       cornerRadii=<4 4 4 4>
      }
      tintColor {colorUseRule=UseColor_Specified, specifiedColor=<rgba 0.0312 0.0312 0.0312 1>}
      uVRegion {bIsValid=false, max=<0 0>, min=<0 0>}
     }
     uncheckedPressedImage {
      bIsDynamicallyLoaded=false
      drawAs=RoundedBox
      imageType=NoImage
      mirroring=NoMirror
      resourceName=None
      resourceObject=null
      tiling=NoTile
      imageSize=<0 0>
      margin=<ltrb 0 0 0 0>
      outlineSettings {
       bUseBrushTransparency=false
       roundingType=FixedRadius
       width=0
       color {colorUseRule=UseColor_Specified, specifiedColor=<rgba 0.0395 0.0395 0.0395 1>}
       cornerRadii=<4 4 4 4>
      }
      tintColor {colorUseRule=UseColor_Specified, specifiedColor=<rgba 0.0395 0.0395 0.0395 1>}
      uVRegion {bIsValid=false, max=<0 0>, min=<0 0>}
     }
     uncheckedSlateSound {resourceObject=null}
     undeterminedForeground {colorUseRule=UseColor_Specified, specifiedColor=<rgba 1 1 1 1>}
     undeterminedHoveredImage {
      bIsDynamicallyLoaded=false
      drawAs=Image
      imageType=Vector
      mirroring=NoMirror
      resourceName=../../../Engine/Content/Slate/Starship/CoreWidgets/CheckBox/indeterminate.svg
      resourceObject=null
      tiling=NoTile
      imageSize=<16 16>
      margin=<ltrb 0 0 0 0>
      outlineSettings {
       bUseBrushTransparency=false
       roundingType=HalfHeightRadius
       width=0
       color {colorUseRule=UseColor_Specified, specifiedColor=<rgba 0 0 0 0>}
       cornerRadii=<0 0 0 0>
      }
      tintColor {colorUseRule=UseColor_Specified, specifiedColor=<rgba 0.0044 0.2384 1 1>}
      uVRegion {bIsValid=false, max=<0 0>, min=<0 0>}
     }
     undeterminedImage {
      bIsDynamicallyLoaded=false
      drawAs=Image
      imageType=Vector
      mirroring=NoMirror
      resourceName=../../../Engine/Content/Slate/Starship/CoreWidgets/CheckBox/indeterminate.svg
      resourceObject=null
      tiling=NoTile
      imageSize=<16 16>
      margin=<ltrb 0 0 0 0>
      outlineSettings {
       bUseBrushTransparency=false
       roundingType=HalfHeightRadius
       width=0
       color {colorUseRule=UseColor_Specified, specifiedColor=<rgba 0 0 0 0>}
       cornerRadii=<0 0 0 0>
      }
      tintColor {colorUseRule=UseColor_Specified, specifiedColor=<rgba 0 0.162 0.7454 1>}
      uVRegion {bIsValid=false, max=<0 0>, min=<0 0>}
     }
     undeterminedPressedImage {
      bIsDynamicallyLoaded=false
      drawAs=Image
      imageType=Vector
      mirroring=NoMirror
      resourceName=../../../Engine/Content/Slate/Starship/CoreWidgets/CheckBox/indeterminate.svg
      resourceObject=null
      tiling=NoTile
      imageSize=<16 16>
      margin=<ltrb 0 0 0 0>
      outlineSettings {
       bUseBrushTransparency=false
       roundingType=HalfHeightRadius
       width=0
       color {colorUseRule=UseColor_Specified, specifiedColor=<rgba 0 0 0 0>}
       cornerRadii=<0 0 0 0>
      }
      tintColor {colorUseRule=UseColor_Specified, specifiedColor=<rgba 0 0.0802 0.3515 1>}
      uVRegion {bIsValid=false, max=<0 0>, min=<0 0>}
     }
    }
   }
   slot {class=SizeBoxSlot}
  }
  TechAudioToolsEditorSlateIcon_RecordIcon {
   widgetClassPath=/Script/TechAudioToolsContentEditor.TechAudioToolsEditorSlateIcon
   bIsVariable=true
   namedSlotHost=WBP_MetaSoundLiteralBoolPrimitive_ToggleButton_Record
   widget=TechAudioToolsEditorSlateIcon_RecordIcon
   properties {fullIconPath=EditorStyle.Icons.filledcircle, size=<32 32>}
  }
  Border_65 {
   widgetClassPath=&Border
   parent=HorizontalBox_20
   widget=Border_65
   properties {horizontalAlignment=HAlign_Center, brushColor=<rgba 0.0176 0.0176 0.0176 1>}
   slot {class=HorizontalBoxSlot, properties {padding=<ltrb 4 2 4 2>}}
  }
  SizeBox_0 {
   widgetClassPath=&SizeBox
   parent=Border_65
   widget=SizeBox_0
   properties {bOverride_MinDesiredWidth=true, minDesiredWidth=500}
   slot {class=BorderSlot, properties {horizontalAlignment=HAlign_Center}}
  }
  Border_70 {
   widgetClassPath=&Border
   parent=SizeBox_0
   widget=Border_70
   properties {brushColor=<rgba 0.0048 0.0048 0.0048 1>}
   slot {class=SizeBoxSlot}
  }
  HorizontalBox_0 {widgetClassPath=&HorizontalBox, parent=Border_70, widget=HorizontalBox_0, slot {class=BorderSlot}}
  TextBlock_78 {
   widgetClassPath=&TextBlock
   parent=HorizontalBox_0
   widget=TextBlock_78
   properties {
    text=File Name:
    font {
     bForceMonospaced=false
     bMaterialIsStencil=false
     fontMaterial=null
     fontObject=&Roboto
     letterSpacing=0
     monospacedWidth=1
     size=18
     skewAmount=0
     typefaceFontName=Light
     outlineSettings {
      bApplyOutlineToDropShadows=false
      bMiteredCorners=false
      bSeparateFillAlpha=false
      outlineMaterial=null
      outlineSize=0
      outlineColor=<rgba 0 0 0 1>
     }
    }
   }
   slot {class=HorizontalBoxSlot, properties {verticalAlignment=VAlign_Center}}
  }
  Spacer_102 {
   widgetClassPath=&Spacer
   parent=HorizontalBox_0
   widget=Spacer_102
   slot {class=HorizontalBoxSlot, properties {padding=<ltrb 10 0 10 0>}}
  }
  WBP_MetaSoundLiteralStringPrimitive_EditableText_SFXGen {
   widgetClassPath=&/WBP_MetaSoundLiteralStringPrimitive_EditableText_SFXGen
   bIsVariable=true
   parent=HorizontalBox_0
   widget=WBP_MetaSoundLiteralStringPrimitive_EditableText_SFXGen
   properties {
    input Name=WaveWriter.FileName
    font {
     bForceMonospaced=false
     bMaterialIsStencil=false
     fontMaterial=null
     fontObject=&Roboto
     letterSpacing=0
     monospacedWidth=1
     size=18
     skewAmount=0
     typefaceFontName=Light
     outlineSettings {
      bApplyOutlineToDropShadows=false
      bMiteredCorners=false
      bSeparateFillAlpha=false
      outlineMaterial=null
      outlineSize=0
      outlineColor=<rgba 0 0 0 1>
     }
    }
   }
   slot {class=HorizontalBoxSlot, properties {verticalAlignment=VAlign_Center, size {sizeRule=Fill, value=1}}}
  }
  EditorUtilityScrollBox_44 {
   widgetClassPath=&EditorUtilityScrollBox
   namedSlotHost=WBP_SFXGeneratorContentPanel_105
   widget=EditorUtilityScrollBox_44
   properties {alwaysShowScrollbarTrack=true, orientation=Orient_Horizontal}
  }
  GridPanel_43 {widgetClassPath=&GridPanel, parent=EditorUtilityScrollBox_44, widget=GridPanel_43, slot {class=ScrollBoxSlot}}
  Border_196 {
   widgetClassPath=&Border
   parent=GridPanel_43
   widget=Border_196
   properties {brushColor=<rgba 0.0176 0.0176 0.0176 1>}
   slot {class=GridSlot, properties {columnSpan=1, rowSpan=2}}
  }
  WBP_SFXGeneratorContentPanel {
   widgetClassPath=&WBP_SFXGeneratorContentPanel
   bIsVariable=true
   parent=Border_196
   widget=WBP_SFXGeneratorContentPanel
   slot {class=BorderSlot}
  }
  TextBlock_97 {widgetClassPath=&TextBlock, namedSlotHost=WBP_SFXGeneratorContentPanel, widget=TextBlock_97, properties {text=Generator}}
  Border_195 {
   widgetClassPath=&Border
   namedSlotHost=WBP_SFXGeneratorContentPanel
   widget=Border_195
   properties {brushColor=<rgba 0.0176 0.0176 0.0176 1>}
  }
  EUW_SFXGenerator_Randomizer_Generator {
   widgetClassPath=&EUW_SFXGenerator_Randomizer
   bIsVariable=true
   parent=Border_195
   widget=EUW_SFXGenerator_Randomizer_Generator
   slot {class=BorderSlot}
  }
  Border_97 {
   widgetClassPath=&Border
   namedSlotHost=WBP_SFXGeneratorContentPanel
   widget=Border_97
   properties {brushColor=<rgba 0.0284 0.0284 0.0284 1>}
  }
  HorizontalBox_244 {
   widgetClassPath=&HorizontalBox
   parent=Border_97
   widget=HorizontalBox_244
   properties {bOverride_Cursor=true}
   slot {class=BorderSlot}
  }
  Border {
   widgetClassPath=&Border
   parent=HorizontalBox_244
   widget=Border
   properties {brushColor=<rgba 0.0176 0.0176 0.0176 1>}
   slot {class=HorizontalBoxSlot}
  }
  WBP_SFXGeneratorContentPanel_1 {
   widgetClassPath=&WBP_SFXGeneratorContentPanel
   bIsVariable=true
   parent=Border
   widget=WBP_SFXGeneratorContentPanel_1
   slot {class=BorderSlot}
  }
  TextBlock {
   widgetClassPath=&TextBlock
   namedSlotHost=WBP_SFXGeneratorContentPanel_1
   widget=TextBlock
   properties {
    text=Oscillator
    font {
     bForceMonospaced=false
     bMaterialIsStencil=false
     fontMaterial=null
     fontObject=&Roboto
     letterSpacing=0
     monospacedWidth=1
     size=18
     skewAmount=0
     typefaceFontName=Bold
     outlineSettings {
      bApplyOutlineToDropShadows=false
      bMiteredCorners=false
      bSeparateFillAlpha=false
      outlineMaterial=null
      outlineSize=0
      outlineColor=<rgba 0 0 0 1>
     }
    }
   }
  }
  Border_2 {
   widgetClassPath=&Border
   namedSlotHost=WBP_SFXGeneratorContentPanel_1
   widget=Border_2
   properties {brushColor=<rgba 0.0176 0.0176 0.0176 1>}
  }
  EUW_SFXGenerator_Randomizer_Osc {
   widgetClassPath=&EUW_SFXGenerator_Randomizer
   bIsVariable=true
   parent=Border_2
   widget=EUW_SFXGenerator_Randomizer_Osc
   slot {class=BorderSlot}
  }
  Border_1 {
   widgetClassPath=&Border
   namedSlotHost=WBP_SFXGeneratorContentPanel_1
   widget=Border_1
   properties {horizontalAlignment=HAlign_Center, verticalAlignment=VAlign_Center, brushColor=<rgba 0.0284 0.0284 0.0284 1>}
  }
  HorizontalBox_245 {
   widgetClassPath=&HorizontalBox
   parent=Border_1
   widget=HorizontalBox_245
   slot {class=BorderSlot, properties {horizontalAlignment=HAlign_Center, verticalAlignment=VAlign_Center}}
  }
  VerticalBox_115 {widgetClassPath=&VerticalBox, parent=HorizontalBox_245, widget=VerticalBox_115, slot {class=HorizontalBoxSlot}}
  WBP_HeaderContentPanel_376 {
   widgetClassPath=/TechAudioToolsContent/Widgets/Layout/WBP_HeaderContentPanel
   bIsVariable=true
   parent=VerticalBox_115
   widget=WBP_HeaderContentPanel_376
   slot {class=VerticalBoxSlot}
  }
  Border_50 {
   widgetClassPath=&Border
   namedSlotHost=WBP_HeaderContentPanel_376
   widget=Border_50
   properties {horizontalAlignment=HAlign_Center, brushColor=<rgba 0.0176 0.0176 0.0176 1>}
  }
  HorizontalBox_17 {
   widgetClassPath=&HorizontalBox
   parent=Border_50
   widget=HorizontalBox_17
   slot {class=BorderSlot, properties {horizontalAlignment=HAlign_Center}}
  }
  EUW_SFXGeneratorKnob_Osc_Frequency {
   widgetClassPath=&/EUW_MetaSoundFloat_Knob_SFXGen
   bIsVariable=true
   parent=HorizontalBox_17
   widget=EUW_SFXGeneratorKnob_Osc_Frequency
   properties {
    input Name=Generator.Pitch.BaseFrequency
    label Text=Frequency
    metaSoundLiteralViewModel_Float=&self:WidgetTree.EUW_SFXGeneratorKnob_Osc_Frequency.MetaSoundLiteralViewModel_Float_0
   }
   slot {class=HorizontalBoxSlot}
  }
  EUW_SFXGeneratorKnob_Osc_Waveform {
   widgetClassPath=&EUW_MetaSoundFloat_Knob_SFXGen
   bIsVariable=true
   parent=HorizontalBox_17
   widget=EUW_SFXGeneratorKnob_Osc_Waveform
   properties {
    input Name=Generator.Waveform
    label Text=Waveform
    metaSoundLiteralViewModel_Float=&self:WidgetTree.EUW_SFXGeneratorKnob_Osc_Waveform.MetaSoundLiteralViewModel_Float_0
   }
   slot {class=HorizontalBoxSlot}
  }
  WBP_HeaderContentPanel_3 {
   widgetClassPath=&WBP_HeaderContentPanel
   bIsVariable=true
   parent=VerticalBox_115
   widget=WBP_HeaderContentPanel_3
   slot {class=VerticalBoxSlot}
  }
  Border_49 {
   widgetClassPath=&Border
   namedSlotHost=WBP_HeaderContentPanel_3
   widget=Border_49
   properties {brushColor=<rgba 0.0176 0.0176 0.0176 1>}
  }
  HorizontalBox_16 {widgetClassPath=&HorizontalBox, parent=Border_49, widget=HorizontalBox_16, slot {class=BorderSlot}}
  EUW_SFXGeneratorKnob_Osc_PW_DutyCycle {
   widgetClassPath=&EUW_MetaSoundFloat_Knob_SFXGen
   bIsVariable=true
   parent=HorizontalBox_16
   widget=EUW_SFXGeneratorKnob_Osc_PW_DutyCycle
   properties {
    default Value=0.5
    input Name=Generator.PW.DutyCycle
    label Text=Duty Cycle
    metaSoundLiteralViewModel_Float=&self:WidgetTree.EUW_SFXGeneratorKnob_Osc_PW_DutyCycle.MetaSoundLiteralViewModel_Float_0
   }
   slot {class=HorizontalBoxSlot}
  }
  EUW_SFXGeneratorKnob_Osc_PW_Sweep {
   widgetClassPath=&EUW_MetaSoundFloat_Knob_SFXGen
   bIsVariable=true
   parent=HorizontalBox_16
   widget=EUW_SFXGeneratorKnob_Osc_PW_Sweep
   properties {
    input Name=Generator.PW.DutySweep
    label Text=Sweep
    metaSoundLiteralViewModel_Float=&self:WidgetTree.EUW_SFXGeneratorKnob_Osc_PW_Sweep.MetaSoundLiteralViewModel_Float_0
   }
   slot {class=HorizontalBoxSlot}
  }
  TextBlock_4 {
   widgetClassPath=&TextBlock
   namedSlotHost=WBP_HeaderContentPanel_3
   widget=TextBlock_4
   properties {
    text=Pulse Width Mod
    font {
     bForceMonospaced=false
     bMaterialIsStencil=false
     fontMaterial=null
     fontObject=&Roboto
     letterSpacing=0
     monospacedWidth=1
     size=12
     skewAmount=0
     typefaceFontName=Bold
     outlineSettings {
      bApplyOutlineToDropShadows=false
      bMiteredCorners=false
      bSeparateFillAlpha=false
      outlineMaterial=null
      outlineSize=0
      outlineColor=<rgba 0 0 0 1>
     }
    }
   }
  }
  Border_51 {
   widgetClassPath=&Border
   parent=HorizontalBox_245
   widget=Border_51
   properties {horizontalAlignment=HAlign_Center, verticalAlignment=VAlign_Center, brushColor=<rgba 0.0176 0.0176 0.0176 1>}
   slot {class=HorizontalBoxSlot}
  }
  WBP_HeaderContentPanel_8 {
   widgetClassPath=&WBP_HeaderContentPanel
   bIsVariable=true
   parent=Border_51
   widget=WBP_HeaderContentPanel_8
   slot {class=BorderSlot, properties {horizontalAlignment=HAlign_Center, verticalAlignment=VAlign_Center}}
  }
  Border_52 {
   widgetClassPath=&Border
   namedSlotHost=WBP_HeaderContentPanel_8
   widget=Border_52
   properties {brushColor=<rgba 0.0176 0.0176 0.0176 1>}
  }
  SizeBox_4 {
   widgetClassPath=&SizeBox
   parent=Border_52
   widget=SizeBox_4
   properties {bOverride_HeightOverride=true, bOverride_WidthOverride=true, heightOverride=150, widthOverride=150}
   slot {class=BorderSlot}
  }
  AudioOscilloscope_2 {
   widgetClassPath=/Script/AudioWidgets.AudioOscilloscope
   bIsVariable=true
   parent=SizeBox_4
   widget=AudioOscilloscope_2
   properties {audioBus=&AB_SFXG_Osc, maxTimeWindowMs=10}
   slot {class=SizeBoxSlot}
  }
  TextBlock_14 {
   widgetClassPath=&TextBlock
   namedSlotHost=WBP_HeaderContentPanel_8
   widget=TextBlock_14
   properties {
    text=Osc View
    font {
     bForceMonospaced=false
     bMaterialIsStencil=false
     fontMaterial=null
     fontObject=&Roboto
     letterSpacing=0
     monospacedWidth=1
     size=12
     skewAmount=0
     typefaceFontName=Bold
     outlineSettings {
      bApplyOutlineToDropShadows=false
      bMiteredCorners=false
      bSeparateFillAlpha=false
      outlineMaterial=null
      outlineSize=0
      outlineColor=<rgba 0 0 0 1>
     }
    }
   }
  }
  Border_3 {
   widgetClassPath=&Border
   parent=HorizontalBox_244
   widget=Border_3
   properties {brushColor=<rgba 0.0176 0.0176 0.0176 1>}
   slot {class=HorizontalBoxSlot}
  }
  WBP_SFXGeneratorContentPanel_2 {
   widgetClassPath=&WBP_SFXGeneratorContentPanel
   bIsVariable=true
   parent=Border_3
   widget=WBP_SFXGeneratorContentPanel_2
   slot {class=BorderSlot}
  }
  TextBlock_1 {
   widgetClassPath=&TextBlock
   namedSlotHost=WBP_SFXGeneratorContentPanel_2
   widget=TextBlock_1
   properties {
    text=Pitch Mod
    font {
     bForceMonospaced=false
     bMaterialIsStencil=false
     fontMaterial=null
     fontObject=&Roboto
     letterSpacing=0
     monospacedWidth=1
     size=18
     skewAmount=0
     typefaceFontName=Bold
     outlineSettings {
      bApplyOutlineToDropShadows=false
      bMiteredCorners=false
      bSeparateFillAlpha=false
      outlineMaterial=null
      outlineSize=0
      outlineColor=<rgba 0 0 0 1>
     }
    }
   }
  }
  Border_5 {
   widgetClassPath=&Border
   namedSlotHost=WBP_SFXGeneratorContentPanel_2
   widget=Border_5
   properties {brushColor=<rgba 0.0176 0.0176 0.0176 1>}
  }
  EUW_SFXGenerator_Randomizer_PitchMod {
   widgetClassPath=&EUW_SFXGenerator_Randomizer
   bIsVariable=true
   parent=Border_5
   widget=EUW_SFXGenerator_Randomizer_PitchMod
   slot {class=BorderSlot}
  }
  Border_4 {
   widgetClassPath=&Border
   namedSlotHost=WBP_SFXGeneratorContentPanel_2
   widget=Border_4
   properties {brushColor=<rgba 0.0284 0.0284 0.0284 1>}
  }
  HorizontalBox_248 {widgetClassPath=&HorizontalBox, parent=Border_4, widget=HorizontalBox_248, slot {class=BorderSlot}}
  VerticalBox_171 {
   widgetClassPath=&VerticalBox
   parent=HorizontalBox_248
   widget=VerticalBox_171
   properties {bOverride_Cursor=true}
   slot {class=HorizontalBoxSlot}
  }
  HorizontalBox_246 {
   widgetClassPath=&HorizontalBox
   parent=VerticalBox_171
   widget=HorizontalBox_246
   properties {bOverride_Cursor=true}
   slot {class=VerticalBoxSlot, properties {verticalAlignment=VAlign_Center, size {sizeRule=Fill, value=1}}}
  }
  Border_333 {
   widgetClassPath=&Border
   parent=HorizontalBox_246
   widget=Border_333
   properties {horizontalAlignment=HAlign_Center, brushColor=<rgba 0.0284 0.0284 0.0284 1>}
   slot {class=HorizontalBoxSlot}
  }
  WBP_HeaderContentPanel_1 {
   widgetClassPath=&WBP_HeaderContentPanel
   bIsVariable=true
   parent=Border_333
   widget=WBP_HeaderContentPanel_1
   slot {class=BorderSlot, properties {horizontalAlignment=HAlign_Center}}
  }
  Border_56 {
   widgetClassPath=&Border
   namedSlotHost=WBP_HeaderContentPanel_1
   widget=Border_56
   properties {brushColor=<rgba 0.0176 0.0176 0.0176 1>}
  }
  HorizontalBox_19 {widgetClassPath=&HorizontalBox, parent=Border_56, widget=HorizontalBox_19, slot {class=BorderSlot}}
  EUW_SFXGeneratorKnob_PM_Sweep_Dir {
   widgetClassPath=&EUW_MetaSoundFloat_Knob_SFXGen
   bIsVariable=true
   parent=HorizontalBox_19
   widget=EUW_SFXGeneratorKnob_PM_Sweep_Dir
   properties {
    default Value=0.5
    input Name=Generator.Pitch.Sweep.Target
    label Text=Direction
    metaSoundLiteralViewModel_Float=&self:WidgetTree.EUW_SFXGeneratorKnob_PM_Sweep_Dir.MetaSoundLiteralViewModel_Float_0
   }
   slot {class=HorizontalBoxSlot}
  }
  EUW_SFXGeneratorKnob_PM_Sweep_Curve {
   widgetClassPath=&EUW_MetaSoundFloat_Knob_SFXGen
   bIsVariable=true
   parent=HorizontalBox_19
   widget=EUW_SFXGeneratorKnob_PM_Sweep_Curve
   properties {
    default Value=0.5
    input Name=Generator.Pitch.Sweep.Curve
    label Text=Curve
    metaSoundLiteralViewModel_Float=&self:WidgetTree.EUW_SFXGeneratorKnob_PM_Sweep_Curve.MetaSoundLiteralViewModel_Float_0
   }
   slot {class=HorizontalBoxSlot}
  }
  TextBlock_2 {
   widgetClassPath=&TextBlock
   namedSlotHost=WBP_HeaderContentPanel_1
   widget=TextBlock_2
   properties {
    text=Sweep
    font {
     bForceMonospaced=false
     bMaterialIsStencil=false
     fontMaterial=null
     fontObject=&Roboto
     letterSpacing=0
     monospacedWidth=1
     size=12
     skewAmount=0
     typefaceFontName=Bold
     outlineSettings {
      bApplyOutlineToDropShadows=false
      bMiteredCorners=false
      bSeparateFillAlpha=false
      outlineMaterial=null
      outlineSize=0
      outlineColor=<rgba 0 0 0 1>
     }
    }
   }
  }
  Border_53 {
   widgetClassPath=&Border
   parent=HorizontalBox_246
   widget=Border_53
   properties {brushColor=<rgba 0.0176 0.0176 0.0176 1>}
   slot {class=HorizontalBoxSlot, properties {size {sizeRule=Fill, value=1}}}
  }
  WBP_HeaderContentPanel_9 {
   widgetClassPath=&WBP_HeaderContentPanel
   bIsVariable=true
   parent=Border_53
   widget=WBP_HeaderContentPanel_9
   slot {class=BorderSlot}
  }
  Border_54 {
   widgetClassPath=&Border
   namedSlotHost=WBP_HeaderContentPanel_9
   widget=Border_54
   properties {horizontalAlignment=HAlign_Center, brushColor=<rgba 0.0176 0.0176 0.0176 1>}
  }
  SizeBox_5 {
   widgetClassPath=&SizeBox
   parent=Border_54
   widget=SizeBox_5
   properties {bOverride_HeightOverride=true, bOverride_MaxDesiredWidth=true, bOverride_WidthOverride=true, maxDesiredWidth=150, widthOverride=150}
   slot {class=BorderSlot, properties {horizontalAlignment=HAlign_Center}}
  }
  AudioOscilloscope_3 {
   widgetClassPath=&AudioOscilloscope
   bIsVariable=true
   parent=SizeBox_5
   widget=AudioOscilloscope_3
   properties {audioBus=&AB_SFXG_PitchBend, maxTimeWindowMs=10, timeWindowMs=1000}
   slot {class=SizeBoxSlot}
  }
  TextBlock_15 {
   widgetClassPath=&TextBlock
   namedSlotHost=WBP_HeaderContentPanel_9
   widget=TextBlock_15
   properties {
    text=Pitch Mod View
    font {
     bForceMonospaced=false
     bMaterialIsStencil=false
     fontMaterial=null
     fontObject=&Roboto
     letterSpacing=0
     monospacedWidth=1
     size=12
     skewAmount=0
     typefaceFontName=Bold
     outlineSettings {
      bApplyOutlineToDropShadows=false
      bMiteredCorners=false
      bSeparateFillAlpha=false
      outlineMaterial=null
      outlineSize=0
      outlineColor=<rgba 0 0 0 1>
     }
    }
   }
  }
  HorizontalBox_247 {
   widgetClassPath=&HorizontalBox
   parent=VerticalBox_171
   widget=HorizontalBox_247
   properties {bOverride_Cursor=true}
   slot {class=VerticalBoxSlot, properties {verticalAlignment=VAlign_Center, size {sizeRule=Fill, value=1}}}
  }
  Border_334 {
   widgetClassPath=&Border
   parent=HorizontalBox_247
   widget=Border_334
   properties {brushColor=<rgba 0.0284 0.0284 0.0284 1>}
   slot {class=HorizontalBoxSlot}
  }
  WBP_HeaderContentPanel {
   widgetClassPath=&WBP_HeaderContentPanel
   bIsVariable=true
   parent=Border_334
   widget=WBP_HeaderContentPanel
   slot {class=BorderSlot}
  }
  Border_55 {
   widgetClassPath=&Border
   namedSlotHost=WBP_HeaderContentPanel
   widget=Border_55
   properties {brushColor=<rgba 0.0176 0.0176 0.0176 1>}
  }
  HorizontalBox_18 {widgetClassPath=&HorizontalBox, parent=Border_55, widget=HorizontalBox_18, slot {class=BorderSlot}}
  EUW_SFXGeneratorKnob_PM_Vibrato_Freq {
   widgetClassPath=&EUW_MetaSoundFloat_Knob_SFXGen
   bIsVariable=true
   parent=HorizontalBox_18
   widget=EUW_SFXGeneratorKnob_PM_Vibrato_Freq
   properties {
    input Name=Generator.Vibrato.Frequency
    label Text=Frequency
    metaSoundLiteralViewModel_Float=&self:WidgetTree.EUW_SFXGeneratorKnob_PM_Vibrato_Freq.MetaSoundLiteralViewModel_Float_0
   }
   slot {class=HorizontalBoxSlot}
  }
  EUW_SFXGeneratorKnob_PM_Vibrato_Depth {
   widgetClassPath=&EUW_MetaSoundFloat_Knob_SFXGen
   bIsVariable=true
   parent=HorizontalBox_18
   widget=EUW_SFXGeneratorKnob_PM_Vibrato_Depth
   properties {
    input Name=Generator.Vibrato.Depth
    label Text=Depth
    metaSoundLiteralViewModel_Float=&self:WidgetTree.EUW_SFXGeneratorKnob_PM_Vibrato_Depth.MetaSoundLiteralViewModel_Float_0
   }
   slot {class=HorizontalBoxSlot}
  }
  TextBlock_161 {
   widgetClassPath=&TextBlock
   namedSlotHost=WBP_HeaderContentPanel
   widget=TextBlock_161
   properties {
    text=Vibrato
    font {
     bForceMonospaced=false
     bMaterialIsStencil=false
     fontMaterial=null
     fontObject=&Roboto
     letterSpacing=0
     monospacedWidth=1
     size=12
     skewAmount=0
     typefaceFontName=Bold
     outlineSettings {
      bApplyOutlineToDropShadows=false
      bMiteredCorners=false
      bSeparateFillAlpha=false
      outlineMaterial=null
      outlineSize=0
      outlineColor=<rgba 0 0 0 1>
     }
    }
   }
  }
  Border_58 {
   widgetClassPath=&Border
   parent=HorizontalBox_247
   widget=Border_58
   properties {brushColor=<rgba 0.0176 0.0176 0.0176 1>}
   slot {class=HorizontalBoxSlot, properties {size {sizeRule=Fill, value=1}}}
  }
  WBP_HeaderContentPanel_10 {
   widgetClassPath=&WBP_HeaderContentPanel
   bIsVariable=true
   parent=Border_58
   widget=WBP_HeaderContentPanel_10
   slot {class=BorderSlot}
  }
  Border_59 {
   widgetClassPath=&Border
   namedSlotHost=WBP_HeaderContentPanel_10
   widget=Border_59
   properties {horizontalAlignment=HAlign_Center, brushColor=<rgba 0.0176 0.0176 0.0176 1>}
  }
  SizeBox_6 {
   widgetClassPath=&SizeBox
   parent=Border_59
   widget=SizeBox_6
   properties {bOverride_HeightOverride=true, bOverride_MaxDesiredWidth=true, bOverride_WidthOverride=true, maxDesiredWidth=150, widthOverride=150}
   slot {class=BorderSlot, properties {horizontalAlignment=HAlign_Center}}
  }
  AudioOscilloscope_4 {
   widgetClassPath=&AudioOscilloscope
   bIsVariable=true
   parent=SizeBox_6
   widget=AudioOscilloscope_4
   properties {audioBus=&AB_SFXG_Vibrato, maxTimeWindowMs=10}
   slot {class=SizeBoxSlot}
  }
  TextBlock_16 {
   widgetClassPath=&TextBlock
   namedSlotHost=WBP_HeaderContentPanel_10
   widget=TextBlock_16
   properties {
    text=Vibrato View
    font {
     bForceMonospaced=false
     bMaterialIsStencil=false
     fontMaterial=null
     fontObject=&Roboto
     letterSpacing=0
     monospacedWidth=1
     size=12
     skewAmount=0
     typefaceFontName=Bold
     outlineSettings {
      bApplyOutlineToDropShadows=false
      bMiteredCorners=false
      bSeparateFillAlpha=false
      outlineMaterial=null
      outlineSize=0
      outlineColor=<rgba 0 0 0 1>
     }
    }
   }
  }
  Border_335 {
   widgetClassPath=&Border
   parent=HorizontalBox_248
   widget=Border_335
   properties {brushColor=<rgba 0.0284 0.0284 0.0284 1>}
   slot {class=HorizontalBoxSlot}
  }
  WBP_HeaderContentPanel_2 {
   widgetClassPath=&WBP_HeaderContentPanel
   bIsVariable=true
   parent=Border_335
   widget=WBP_HeaderContentPanel_2
   slot {class=BorderSlot}
  }
  Border_57 {
   widgetClassPath=&Border
   namedSlotHost=WBP_HeaderContentPanel_2
   widget=Border_57
   properties {brushColor=<rgba 0.0176 0.0176 0.0176 1>}
  }
  HorizontalBox_249 {widgetClassPath=&HorizontalBox, parent=Border_57, widget=HorizontalBox_249, slot {class=BorderSlot}}
  HorizontalBox_250 {
   widgetClassPath=&HorizontalBox
   parent=HorizontalBox_249
   widget=HorizontalBox_250
   properties {bOverride_Cursor=true}
   slot {class=HorizontalBoxSlot}
  }
  Border_336 {
   widgetClassPath=&Border
   parent=HorizontalBox_250
   widget=Border_336
   properties {verticalAlignment=VAlign_Center, brushColor=<rgba 0.0176 0.0176 0.0176 1>}
   slot {class=HorizontalBoxSlot}
  }
  EUW_MetaSoundInteger_Slider_Vertical_SFXGen_PM_Jump_Num {
   widgetClassPath=&/EUW_MetaSoundInteger_Slider_Vertical_SFXGen
   bIsVariable=true
   parent=Border_336
   widget=EUW_MetaSoundInteger_Slider_Vertical_SFXGen_PM_Jump_Num
   properties {
    input Name=Generator.Pitch.Jump.Num
    label Text=Number
    metaSoundLiteralViewModel_Integer=&self:WidgetTree.EUW_MetaSoundInteger_Slider_Vertical_SFXGen_PM_Jump_Num.MetaSoundLiteralViewModel_Integer_0
   }
   slot {class=BorderSlot, properties {verticalAlignment=VAlign_Center}}
  }
  Border_337 {
   widgetClassPath=&Border
   parent=HorizontalBox_250
   widget=Border_337
   properties {verticalAlignment=VAlign_Center, brushColor=<rgba 0.0176 0.0176 0.0176 1>}
   slot {class=HorizontalBoxSlot}
  }
  VerticalBox_173 {
   widgetClassPath=&VerticalBox
   parent=Border_337
   widget=VerticalBox_173
   properties {bOverride_Cursor=true}
   slot {class=BorderSlot, properties {verticalAlignment=VAlign_Center}}
  }
  EUW_SFXGeneratorKnob_PM_Jump_Size {
   widgetClassPath=&EUW_MetaSoundFloat_Knob_SFXGen
   bIsVariable=true
   parent=VerticalBox_173
   widget=EUW_SFXGeneratorKnob_PM_Jump_Size
   properties {
    input Name=Generator.Pitch.Jump.Size
    label Text=Size
    metaSoundLiteralViewModel_Float=&self:WidgetTree.EUW_SFXGeneratorKnob_PM_Jump_Size.MetaSoundLiteralViewModel_Float_0
   }
   slot {class=VerticalBoxSlot, properties {size {sizeRule=Fill, value=1}}}
  }
  EUW_SFXGeneratorKnob_PM_Jump_Speed {
   widgetClassPath=&EUW_MetaSoundFloat_Knob_SFXGen
   bIsVariable=true
   parent=VerticalBox_173
   widget=EUW_SFXGeneratorKnob_PM_Jump_Speed
   properties {
    input Name=Generator.Pitch.Jump.Speed
    label Text=Speed
    metaSoundLiteralViewModel_Float=&self:WidgetTree.EUW_SFXGeneratorKnob_PM_Jump_Speed.MetaSoundLiteralViewModel_Float_0
   }
   slot {class=VerticalBoxSlot, properties {size {sizeRule=Fill, value=1}}}
  }
  TextBlock_3 {
   widgetClassPath=&TextBlock
   namedSlotHost=WBP_HeaderContentPanel_2
   widget=TextBlock_3
   properties {
    text=Jump
    font {
     bForceMonospaced=false
     bMaterialIsStencil=false
     fontMaterial=null
     fontObject=&Roboto
     letterSpacing=0
     monospacedWidth=1
     size=12
     skewAmount=0
     typefaceFontName=Bold
     outlineSettings {
      bApplyOutlineToDropShadows=false
      bMiteredCorners=false
      bSeparateFillAlpha=false
      outlineMaterial=null
      outlineSize=0
      outlineColor=<rgba 0 0 0 1>
     }
    }
   }
  }
  Border_6 {
   widgetClassPath=&Border
   parent=GridPanel_43
   widget=Border_6
   properties {brushColor=<rgba 0.0176 0.0176 0.0176 1>}
   slot {class=GridSlot, properties {column=1, columnSpan=1, rowSpan=2}}
  }
  WBP_SFXGeneratorContentPanel_3 {
   widgetClassPath=&WBP_SFXGeneratorContentPanel
   bIsVariable=true
   parent=Border_6
   widget=WBP_SFXGeneratorContentPanel_3
   slot {class=BorderSlot}
  }
  TextBlock_11 {widgetClassPath=&TextBlock, namedSlotHost=WBP_SFXGeneratorContentPanel_3, widget=TextBlock_11, properties {text=Spectral FX}}
  Border_14 {
   widgetClassPath=&Border
   namedSlotHost=WBP_SFXGeneratorContentPanel_3
   widget=Border_14
   properties {brushColor=<rgba 0.0176 0.0176 0.0176 1>}
  }
  EUW_SFXGenerator_Randomizer_SpectralFX {
   widgetClassPath=&EUW_SFXGenerator_Randomizer
   bIsVariable=true
   parent=Border_14
   widget=EUW_SFXGenerator_Randomizer_SpectralFX
   slot {class=BorderSlot}
  }
  Border_7 {
   widgetClassPath=&Border
   namedSlotHost=WBP_SFXGeneratorContentPanel_3
   widget=Border_7
   properties {brushColor=<rgba 0.0284 0.0284 0.0284 1>}
  }
  VerticalBox_2 {widgetClassPath=&VerticalBox, parent=Border_7, widget=VerticalBox_2, slot {class=BorderSlot}}
  Border_199 {
   widgetClassPath=&Border
   parent=VerticalBox_2
   widget=Border_199
   properties {brushColor=<rgba 0.0176 0.0176 0.0176 1>}
   slot {class=VerticalBoxSlot}
  }
  HorizontalBox_59 {widgetClassPath=&HorizontalBox, parent=Border_199, widget=HorizontalBox_59, slot {class=BorderSlot}}
  Border_8 {
   widgetClassPath=&Border
   parent=HorizontalBox_59
   widget=Border_8
   properties {brushColor=<rgba 0.0284 0.0284 0.0284 1>}
   slot {class=HorizontalBoxSlot}
  }
  WBP_SFXGeneratorContentPanel_4 {
   widgetClassPath=&WBP_SFXGeneratorContentPanel
   bIsVariable=true
   parent=Border_8
   widget=WBP_SFXGeneratorContentPanel_4
   slot {class=BorderSlot}
  }
  TextBlock_6 {
   widgetClassPath=&TextBlock
   namedSlotHost=WBP_SFXGeneratorContentPanel_4
   widget=TextBlock_6
   properties {
    text=Wave Shaper
    font {
     bForceMonospaced=false
     bMaterialIsStencil=false
     fontMaterial=null
     fontObject=&Roboto
     letterSpacing=0
     monospacedWidth=1
     size=18
     skewAmount=0
     typefaceFontName=Bold
     outlineSettings {
      bApplyOutlineToDropShadows=false
      bMiteredCorners=false
      bSeparateFillAlpha=false
      outlineMaterial=null
      outlineSize=0
      outlineColor=<rgba 0 0 0 1>
     }
    }
   }
  }
  SizeBox_1 {
   widgetClassPath=&SizeBox
   namedSlotHost=WBP_SFXGeneratorContentPanel_4
   widget=SizeBox_1
   properties {bOverride_MinDesiredHeight=true, minDesiredHeight=32}
  }
  Spacer_137 {widgetClassPath=&Spacer, parent=SizeBox_1, widget=Spacer_137, slot {class=SizeBoxSlot}}
  Border_9 {
   widgetClassPath=&Border
   namedSlotHost=WBP_SFXGeneratorContentPanel_4
   widget=Border_9
   properties {brushColor=<rgba 0.0176 0.0176 0.0176 1>}
  }
  HorizontalBox_2 {widgetClassPath=&HorizontalBox, parent=Border_9, widget=HorizontalBox_2, slot {class=BorderSlot}}
  Border_200 {
   widgetClassPath=&Border
   parent=HorizontalBox_2
   widget=Border_200
   properties {brushColor=<rgba 0.0284 0.0284 0.0284 1>}
   slot {class=HorizontalBoxSlot, properties {size {sizeRule=Fill, value=1}}}
  }
  EUW_SFXGeneratorKnob_SpectralFX_WaveShaper {
   widgetClassPath=&EUW_MetaSoundFloat_Knob_SFXGen
   bIsVariable=true
   parent=Border_200
   widget=EUW_SFXGeneratorKnob_SpectralFX_WaveShaper
   properties {
    input Name=FX.WaveShaper.Amount
    label Text=Amount
    metaSoundLiteralViewModel_Float=&self:WidgetTree.EUW_SFXGeneratorKnob_SpectralFX_WaveShaper.MetaSoundLiteralViewModel_Float_0
   }
   slot {class=BorderSlot}
  }
  Border_10 {
   widgetClassPath=&Border
   parent=HorizontalBox_59
   widget=Border_10
   properties {brushColor=<rgba 0.0176 0.0176 0.0176 1>}
   slot {class=HorizontalBoxSlot, properties {size {sizeRule=Fill, value=1}}}
  }
  WBP_HeaderContentPanel_5 {
   widgetClassPath=&WBP_HeaderContentPanel
   bIsVariable=true
   parent=Border_10
   widget=WBP_HeaderContentPanel_5
   slot {class=BorderSlot}
  }
  Border_45 {
   widgetClassPath=&Border
   namedSlotHost=WBP_HeaderContentPanel_5
   widget=Border_45
   properties {brushColor=<rgba 0.0176 0.0176 0.0176 1>}
  }
  SizeBox {
   widgetClassPath=&SizeBox
   parent=Border_45
   widget=SizeBox
   properties {bOverride_HeightOverride=true, bOverride_WidthOverride=true, widthOverride=150}
   slot {class=BorderSlot}
  }
  AudioOscilloscope {
   widgetClassPath=&AudioOscilloscope
   bIsVariable=true
   parent=SizeBox
   widget=AudioOscilloscope
   properties {audioBus=&AB_SFXG_WaveShaper, maxTimeWindowMs=10}
   slot {class=SizeBoxSlot}
  }
  TextBlock_7 {
   widgetClassPath=&TextBlock
   namedSlotHost=WBP_HeaderContentPanel_5
   widget=TextBlock_7
   properties {
    text=WS View
    font {
     bForceMonospaced=false
     bMaterialIsStencil=false
     fontMaterial=null
     fontObject=&Roboto
     letterSpacing=0
     monospacedWidth=1
     size=12
     skewAmount=0
     typefaceFontName=Bold
     outlineSettings {
      bApplyOutlineToDropShadows=false
      bMiteredCorners=false
      bSeparateFillAlpha=false
      outlineMaterial=null
      outlineSize=0
      outlineColor=<rgba 0 0 0 1>
     }
    }
   }
  }
  Border_11 {
   widgetClassPath=&Border
   parent=HorizontalBox_59
   widget=Border_11
   properties {brushColor=<rgba 0.0284 0.0284 0.0284 1>}
   slot {class=HorizontalBoxSlot}
  }
  WBP_SFXGeneratorContentPanel_5 {
   widgetClassPath=&WBP_SFXGeneratorContentPanel
   bIsVariable=true
   parent=Border_11
   widget=WBP_SFXGeneratorContentPanel_5
   slot {class=BorderSlot}
  }
  TextBlock_8 {
   widgetClassPath=&TextBlock
   namedSlotHost=WBP_SFXGeneratorContentPanel_5
   widget=TextBlock_8
   properties {
    text=Bit Crusher
    font {
     bForceMonospaced=false
     bMaterialIsStencil=false
     fontMaterial=null
     fontObject=&Roboto
     letterSpacing=0
     monospacedWidth=1
     size=18
     skewAmount=0
     typefaceFontName=Bold
     outlineSettings {
      bApplyOutlineToDropShadows=false
      bMiteredCorners=false
      bSeparateFillAlpha=false
      outlineMaterial=null
      outlineSize=0
      outlineColor=<rgba 0 0 0 1>
     }
    }
   }
  }
  Border_13 {
   widgetClassPath=&Border
   namedSlotHost=WBP_SFXGeneratorContentPanel_5
   widget=Border_13
   properties {brushColor=<rgba 0.0176 0.0176 0.0176 1>}
  }
  EUW_SFXGenerator_Randomizer_BitCrusher {
   widgetClassPath=&EUW_SFXGenerator_Randomizer
   bIsVariable=true
   parent=Border_13
   widget=EUW_SFXGenerator_Randomizer_BitCrusher
   slot {class=BorderSlot}
  }
  Border_12 {
   widgetClassPath=&Border
   namedSlotHost=WBP_SFXGeneratorContentPanel_5
   widget=Border_12
   properties {brushColor=<rgba 0.0176 0.0176 0.0176 1>}
  }
  HorizontalBox_62 {widgetClassPath=&HorizontalBox, parent=Border_12, widget=HorizontalBox_62, slot {class=BorderSlot}}
  Border_201 {
   widgetClassPath=&Border
   parent=HorizontalBox_62
   widget=Border_201
   properties {brushColor=<rgba 0.0284 0.0284 0.0284 1>}
   slot {class=HorizontalBoxSlot, properties {size {sizeRule=Fill, value=1}}}
  }
  EUW_SFXGeneratorKnob_SpectralFX_BitCrusher_Amount {
   widgetClassPath=&EUW_MetaSoundFloat_Knob_SFXGen
   bIsVariable=true
   parent=Border_201
   widget=EUW_SFXGeneratorKnob_SpectralFX_BitCrusher_Amount
   properties {
    input Name=FX.BitCrusher.Amount
    label Text=Amount
    metaSoundLiteralViewModel_Float=&self:WidgetTree.EUW_SFXGeneratorKnob_SpectralFX_BitCrusher_Amount.MetaSoundLiteralViewModel_Float_0
   }
   slot {class=BorderSlot}
  }
  Border_202 {
   widgetClassPath=&Border
   parent=HorizontalBox_62
   widget=Border_202
   properties {brushColor=<rgba 0.0284 0.0284 0.0284 1>}
   slot {class=HorizontalBoxSlot, properties {size {sizeRule=Fill, value=1}}}
  }
  EUW_SFXGeneratorKnob_SpectralFX_BitCrusher_SampleRate {
   widgetClassPath=&EUW_MetaSoundFloat_Knob_SFXGen
   bIsVariable=true
   parent=Border_202
   widget=EUW_SFXGeneratorKnob_SpectralFX_BitCrusher_SampleRate
   properties {
    default Value=1
    input Name=Amp.BitCrusher.SampleRate
    label Text=Sample Rate
    metaSoundLiteralViewModel_Float=&self:WidgetTree.EUW_SFXGeneratorKnob_SpectralFX_BitCrusher_SampleRate.MetaSoundLiteralViewModel_Float_0
   }
   slot {class=BorderSlot}
  }
  Border_46 {
   widgetClassPath=&Border
   parent=HorizontalBox_59
   widget=Border_46
   properties {brushColor=<rgba 0.0176 0.0176 0.0176 1>}
   slot {class=HorizontalBoxSlot, properties {size {sizeRule=Fill, value=1}}}
  }
  WBP_HeaderContentPanel_7 {
   widgetClassPath=&WBP_HeaderContentPanel
   bIsVariable=true
   parent=Border_46
   widget=WBP_HeaderContentPanel_7
   slot {class=BorderSlot}
  }
  Border_47 {
   widgetClassPath=&Border
   namedSlotHost=WBP_HeaderContentPanel_7
   widget=Border_47
   properties {brushColor=<rgba 0.0176 0.0176 0.0176 1>}
  }
  SizeBox_3 {
   widgetClassPath=&SizeBox
   parent=Border_47
   widget=SizeBox_3
   properties {bOverride_HeightOverride=true, bOverride_WidthOverride=true, widthOverride=150}
   slot {class=BorderSlot}
  }
  AudioOscilloscope_1 {
   widgetClassPath=&AudioOscilloscope
   bIsVariable=true
   parent=SizeBox_3
   widget=AudioOscilloscope_1
   properties {audioBus=&AB_SFXG_BitCrusher, maxTimeWindowMs=10}
   slot {class=SizeBoxSlot}
  }
  TextBlock_12 {
   widgetClassPath=&TextBlock
   namedSlotHost=WBP_HeaderContentPanel_7
   widget=TextBlock_12
   properties {
    text=BC View
    font {
     bForceMonospaced=false
     bMaterialIsStencil=false
     fontMaterial=null
     fontObject=&Roboto
     letterSpacing=0
     monospacedWidth=1
     size=12
     skewAmount=0
     typefaceFontName=Bold
     outlineSettings {
      bApplyOutlineToDropShadows=false
      bMiteredCorners=false
      bSeparateFillAlpha=false
      outlineMaterial=null
      outlineSize=0
      outlineColor=<rgba 0 0 0 1>
     }
    }
   }
  }
  Border_42 {
   widgetClassPath=&Border
   parent=VerticalBox_2
   widget=Border_42
   properties {brushColor=<rgba 0.0176 0.0176 0.0176 1>}
   slot {class=VerticalBoxSlot}
  }
  WBP_SFXGeneratorContentPanel_15 {
   widgetClassPath=&WBP_SFXGeneratorContentPanel
   bIsVariable=true
   parent=Border_42
   widget=WBP_SFXGeneratorContentPanel_15
   slot {class=BorderSlot}
  }
  TextBlock_10 {
   widgetClassPath=&TextBlock
   namedSlotHost=WBP_SFXGeneratorContentPanel_15
   widget=TextBlock_10
   properties {
    text=RingMod
    font {
     bForceMonospaced=false
     bMaterialIsStencil=false
     fontMaterial=null
     fontObject=&Roboto
     letterSpacing=0
     monospacedWidth=1
     size=18
     skewAmount=0
     typefaceFontName=Bold
     outlineSettings {
      bApplyOutlineToDropShadows=false
      bMiteredCorners=false
      bSeparateFillAlpha=false
      outlineMaterial=null
      outlineSize=0
      outlineColor=<rgba 0 0 0 1>
     }
    }
   }
  }
  Border_44 {
   widgetClassPath=&Border
   namedSlotHost=WBP_SFXGeneratorContentPanel_15
   widget=Border_44
   properties {brushColor=<rgba 0.0176 0.0176 0.0176 1>}
  }
  EUW_SFXGenerator_Randomizer_RingMod {
   widgetClassPath=&EUW_SFXGenerator_Randomizer
   bIsVariable=true
   parent=Border_44
   widget=EUW_SFXGenerator_Randomizer_RingMod
   slot {class=BorderSlot}
  }
  Border_43 {
   widgetClassPath=&Border
   namedSlotHost=WBP_SFXGeneratorContentPanel_15
   widget=Border_43
   properties {brushColor=<rgba 0.0284 0.0284 0.0284 1>}
  }
  VerticalBox_75 {
   widgetClassPath=&VerticalBox
   parent=Border_43
   widget=VerticalBox_75
   properties {bOverride_Cursor=true}
   slot {class=BorderSlot}
  }
  HorizontalBox_178 {widgetClassPath=&HorizontalBox, parent=VerticalBox_75, widget=HorizontalBox_178, slot {class=VerticalBoxSlot}}
  WBP_HeaderContentPanel_6 {
   widgetClassPath=&WBP_HeaderContentPanel
   bIsVariable=true
   parent=HorizontalBox_178
   widget=WBP_HeaderContentPanel_6
   slot {class=HorizontalBoxSlot}
  }
  Border_198 {
   widgetClassPath=&Border
   namedSlotHost=WBP_HeaderContentPanel_6
   widget=Border_198
   properties {brushColor=<rgba 0.0176 0.0176 0.0176 1>}
  }
  HorizontalBox_118 {widgetClassPath=&HorizontalBox, parent=Border_198, widget=HorizontalBox_118, slot {class=BorderSlot}}
  EUW_SFXGeneratorKnob_RingMod_Control_Amount {
   widgetClassPath=&EUW_MetaSoundFloat_Knob_SFXGen
   bIsVariable=true
   parent=HorizontalBox_118
   widget=EUW_SFXGeneratorKnob_RingMod_Control_Amount
   properties {
    input Name=FX.Ringmod.Amount
    label Text=Amount
    metaSoundLiteralViewModel_Float=&self:WidgetTree.EUW_SFXGeneratorKnob_RingMod_Control_Amount.MetaSoundLiteralViewModel_Float_0
   }
   slot {class=HorizontalBoxSlot}
  }
  EUW_SFXGeneratorKnob_RingMod_Control_Frequency {
   widgetClassPath=&EUW_MetaSoundFloat_Knob_SFXGen
   bIsVariable=true
   parent=HorizontalBox_118
   widget=EUW_SFXGeneratorKnob_RingMod_Control_Frequency
   properties {
    input Name=FX.RingmodMod.Frequency
    label Text=Frequency
    metaSoundLiteralViewModel_Float=&self:WidgetTree.EUW_SFXGeneratorKnob_RingMod_Control_Frequency.MetaSoundLiteralViewModel_Float_0
   }
   slot {class=HorizontalBoxSlot}
  }
  TextBlock_9 {
   widgetClassPath=&TextBlock
   namedSlotHost=WBP_HeaderContentPanel_6
   widget=TextBlock_9
   properties {
    text=Control
    font {
     bForceMonospaced=false
     bMaterialIsStencil=false
     fontMaterial=null
     fontObject=&Roboto
     letterSpacing=0
     monospacedWidth=1
     size=12
     skewAmount=0
     typefaceFontName=Bold
     outlineSettings {
      bApplyOutlineToDropShadows=false
      bMiteredCorners=false
      bSeparateFillAlpha=false
      outlineMaterial=null
      outlineSize=0
      outlineColor=<rgba 0 0 0 1>
     }
    }
   }
  }
  Border_291 {
   widgetClassPath=&Border
   parent=HorizontalBox_178
   widget=Border_291
   properties {brushColor=<rgba 0.0176 0.0176 0.0176 1>}
   slot {class=HorizontalBoxSlot, properties {size {sizeRule=Fill, value=1}}}
  }
  WBP_HeaderContentPanel_271 {
   widgetClassPath=&WBP_HeaderContentPanel
   bIsVariable=true
   parent=Border_291
   widget=WBP_HeaderContentPanel_271
   slot {class=BorderSlot}
  }
  Border_289 {
   widgetClassPath=&Border
   namedSlotHost=WBP_HeaderContentPanel_271
   widget=Border_289
   properties {brushColor=<rgba 0.0176 0.0176 0.0176 1>}
  }
  SizeBox_2 {
   widgetClassPath=&SizeBox
   parent=Border_289
   widget=SizeBox_2
   properties {bOverride_HeightOverride=true, bOverride_WidthOverride=true}
   slot {class=BorderSlot}
  }
  AudioOscilloscope_78 {
   widgetClassPath=&AudioOscilloscope
   bIsVariable=true
   parent=SizeBox_2
   widget=AudioOscilloscope_78
   properties {audioBus=&AB_SFXG_RingMod, maxTimeWindowMs=10}
   slot {class=SizeBoxSlot}
  }
  TextBlock_303 {
   widgetClassPath=&TextBlock
   namedSlotHost=WBP_HeaderContentPanel_271
   widget=TextBlock_303
   properties {
    text=RM View
    font {
     bForceMonospaced=false
     bMaterialIsStencil=false
     fontMaterial=null
     fontObject=&Roboto
     letterSpacing=0
     monospacedWidth=1
     size=12
     skewAmount=0
     typefaceFontName=Bold
     outlineSettings {
      bApplyOutlineToDropShadows=false
      bMiteredCorners=false
      bSeparateFillAlpha=false
      outlineMaterial=null
      outlineSize=0
      outlineColor=<rgba 0 0 0 1>
     }
    }
   }
  }
  WBP_HeaderContentPanel_4 {
   widgetClassPath=&WBP_HeaderContentPanel
   bIsVariable=true
   parent=HorizontalBox_178
   widget=WBP_HeaderContentPanel_4
   slot {class=HorizontalBoxSlot}
  }
  HorizontalBox_177 {widgetClassPath=&HorizontalBox, namedSlotHost=WBP_HeaderContentPanel_4, widget=HorizontalBox_177}
  EUW_SFXGeneratorKnob_RingMod_Mod_Depth {
   widgetClassPath=&EUW_MetaSoundFloat_Knob_SFXGen
   bIsVariable=true
   parent=HorizontalBox_177
   widget=EUW_SFXGeneratorKnob_RingMod_Mod_Depth
   properties {
    input Name=FX.Ringmod.Envelope.Depth
    label Text=Depth
    metaSoundLiteralViewModel_Float=&self:WidgetTree.EUW_SFXGeneratorKnob_RingMod_Mod_Depth.MetaSoundLiteralViewModel_Float_0
   }
   slot {class=HorizontalBoxSlot}
  }
  EUW_SFXGeneratorKnob_RingMod_Mod_Shape {
   widgetClassPath=&EUW_MetaSoundFloat_Knob_SFXGen
   bIsVariable=true
   parent=HorizontalBox_177
   widget=EUW_SFXGeneratorKnob_RingMod_Mod_Shape
   properties {
    default Value=0.5
    input Name=FX.Ringmod.Envelope.Shape
    label Text=Shape
    metaSoundLiteralViewModel_Float=&self:WidgetTree.EUW_SFXGeneratorKnob_RingMod_Mod_Shape.MetaSoundLiteralViewModel_Float_0
   }
   slot {class=HorizontalBoxSlot}
  }
  EUW_SFXGeneratorKnob_RingMod_Mod_Curve {
   widgetClassPath=&EUW_MetaSoundFloat_Knob_SFXGen
   bIsVariable=true
   parent=HorizontalBox_177
   widget=EUW_SFXGeneratorKnob_RingMod_Mod_Curve
   properties {
    default Value=0.5
    input Name=FX.Ringmod.Envelope.Curve
    label Text=Curve
    metaSoundLiteralViewModel_Float=&self:WidgetTree.EUW_SFXGeneratorKnob_RingMod_Mod_Curve.MetaSoundLiteralViewModel_Float_0
   }
   slot {class=HorizontalBoxSlot}
  }
  TextBlock_5 {
   widgetClassPath=&TextBlock
   namedSlotHost=WBP_HeaderContentPanel_4
   widget=TextBlock_5
   properties {
    text=Modulation
    font {
     bForceMonospaced=false
     bMaterialIsStencil=false
     fontMaterial=null
     fontObject=&Roboto
     letterSpacing=0
     monospacedWidth=1
     size=12
     skewAmount=0
     typefaceFontName=Bold
     outlineSettings {
      bApplyOutlineToDropShadows=false
      bMiteredCorners=false
      bSeparateFillAlpha=false
      outlineMaterial=null
      outlineSize=0
      outlineColor=<rgba 0 0 0 1>
     }
    }
   }
  }
  Border_15 {
   widgetClassPath=&Border
   parent=GridPanel_43
   widget=Border_15
   properties {brushColor=<rgba 0.0176 0.0176 0.0176 1>}
   slot {class=GridSlot, properties {column=2, columnSpan=2, rowSpan=1}}
  }
  WBP_SFXGeneratorContentPanel_6 {
   widgetClassPath=&WBP_SFXGeneratorContentPanel
   bIsVariable=true
   parent=Border_15
   widget=WBP_SFXGeneratorContentPanel_6
   slot {class=BorderSlot}
  }
  TextBlock_18 {widgetClassPath=&TextBlock, namedSlotHost=WBP_SFXGeneratorContentPanel_6, widget=TextBlock_18, properties {text=Filter}}
  Border_23 {
   widgetClassPath=&Border
   namedSlotHost=WBP_SFXGeneratorContentPanel_6
   widget=Border_23
   properties {brushColor=<rgba 0.0176 0.0176 0.0176 1>}
  }
  EUW_SFXGenerator_Randomizer_Filter {
   widgetClassPath=&EUW_SFXGenerator_Randomizer
   bIsVariable=true
   parent=Border_23
   widget=EUW_SFXGenerator_Randomizer_Filter
   slot {class=BorderSlot}
  }
  Border_16 {
   widgetClassPath=&Border
   namedSlotHost=WBP_SFXGeneratorContentPanel_6
   widget=Border_16
   properties {brushColor=<rgba 0.0284 0.0284 0.0284 1>}
  }
  HorizontalBox_253 {
   widgetClassPath=&HorizontalBox
   parent=Border_16
   widget=HorizontalBox_253
   properties {bOverride_Cursor=true}
   slot {class=BorderSlot}
  }
  Border_17 {
   widgetClassPath=&Border
   parent=HorizontalBox_253
   widget=Border_17
   properties {brushColor=<rgba 0.0176 0.0176 0.0176 1>}
   slot {class=HorizontalBoxSlot, properties {horizontalAlignment=HAlign_Center, size {sizeRule=Fill, value=1}}}
  }
  WBP_SFXGeneratorContentPanel_7 {
   widgetClassPath=&WBP_SFXGeneratorContentPanel
   bIsVariable=true
   parent=Border_17
   widget=WBP_SFXGeneratorContentPanel_7
   slot {class=BorderSlot}
  }
  TextBlock_13 {
   widgetClassPath=&TextBlock
   namedSlotHost=WBP_SFXGeneratorContentPanel_7
   widget=TextBlock_13
   properties {
    text=Control
    font {
     bForceMonospaced=false
     bMaterialIsStencil=false
     fontMaterial=null
     fontObject=&Roboto
     letterSpacing=0
     monospacedWidth=1
     size=18
     skewAmount=0
     typefaceFontName=Bold
     outlineSettings {
      bApplyOutlineToDropShadows=false
      bMiteredCorners=false
      bSeparateFillAlpha=false
      outlineMaterial=null
      outlineSize=0
      outlineColor=<rgba 0 0 0 1>
     }
    }
   }
  }
  Border_19 {
   widgetClassPath=&Border
   namedSlotHost=WBP_SFXGeneratorContentPanel_7
   widget=Border_19
   properties {brushColor=<rgba 0.0176 0.0176 0.0176 1>}
  }
  EUW_SFXGenerator_Randomizer_FilterControl {
   widgetClassPath=&EUW_SFXGenerator_Randomizer
   bIsVariable=true
   parent=Border_19
   widget=EUW_SFXGenerator_Randomizer_FilterControl
   slot {class=BorderSlot}
  }
  Border_18 {
   widgetClassPath=&Border
   namedSlotHost=WBP_SFXGeneratorContentPanel_7
   widget=Border_18
   properties {brushColor=<rgba 0.0284 0.0284 0.0284 1>}
  }
  HorizontalBox_251 {
   widgetClassPath=&HorizontalBox
   parent=Border_18
   widget=HorizontalBox_251
   properties {bOverride_Cursor=true}
   slot {class=BorderSlot}
  }
  EUW_SFXGeneratorKnob_Filter_Control_Type {
   widgetClassPath=&EUW_MetaSoundFloat_Knob_SFXGen
   bIsVariable=true
   parent=HorizontalBox_251
   widget=EUW_SFXGeneratorKnob_Filter_Control_Type
   properties {
    input Name=FX.Filter.Type
    label Text=Type
    metaSoundLiteralViewModel_Float=&self:WidgetTree.EUW_SFXGeneratorKnob_Filter_Control_Type.MetaSoundLiteralViewModel_Float_0
   }
   slot {class=HorizontalBoxSlot, properties {size {sizeRule=Fill, value=1}}}
  }
  EUW_SFXGeneratorKnob_Filter_Control_Cutoff {
   widgetClassPath=&EUW_MetaSoundFloat_Knob_SFXGen
   bIsVariable=true
   parent=HorizontalBox_251
   widget=EUW_SFXGeneratorKnob_Filter_Control_Cutoff
   properties {
    default Value=1
    input Name=FX.Filter.Cutoff
    label Text=Cutoff
    metaSoundLiteralViewModel_Float=&self:WidgetTree.EUW_SFXGeneratorKnob_Filter_Control_Cutoff.MetaSoundLiteralViewModel_Float_0
   }
   slot {class=HorizontalBoxSlot, properties {size {sizeRule=Fill, value=1}}}
  }
  EUW_SFXGeneratorKnob_Filter_Control_Resonance {
   widgetClassPath=&EUW_MetaSoundFloat_Knob_SFXGen
   bIsVariable=true
   parent=HorizontalBox_251
   widget=EUW_SFXGeneratorKnob_Filter_Control_Resonance
   properties {
    input Name=FX.Filter.Resonance
    label Text=Resonance
    metaSoundLiteralViewModel_Float=&self:WidgetTree.EUW_SFXGeneratorKnob_Filter_Control_Resonance.MetaSoundLiteralViewModel_Float_0
   }
   slot {class=HorizontalBoxSlot, properties {size {sizeRule=Fill, value=1}}}
  }
  Border_60 {
   widgetClassPath=&Border
   parent=HorizontalBox_253
   widget=Border_60
   properties {brushColor=<rgba 0.0176 0.0176 0.0176 1>}
   slot {class=HorizontalBoxSlot, properties {size {sizeRule=Fill, value=1}}}
  }
  WBP_HeaderContentPanel_11 {
   widgetClassPath=&WBP_HeaderContentPanel
   bIsVariable=true
   parent=Border_60
   widget=WBP_HeaderContentPanel_11
   slot {class=BorderSlot}
  }
  Border_61 {
   widgetClassPath=&Border
   namedSlotHost=WBP_HeaderContentPanel_11
   widget=Border_61
   properties {brushColor=<rgba 0.0176 0.0176 0.0176 1>}
  }
  SizeBox_7 {
   widgetClassPath=&SizeBox
   parent=Border_61
   widget=SizeBox_7
   properties {bOverride_HeightOverride=true, bOverride_WidthOverride=true, widthOverride=150}
   slot {class=BorderSlot}
  }
  AudioOscilloscope_5 {
   widgetClassPath=&AudioOscilloscope
   bIsVariable=true
   parent=SizeBox_7
   widget=AudioOscilloscope_5
   properties {audioBus=&AB_SFXG_Filter, maxTimeWindowMs=10}
   slot {class=SizeBoxSlot}
  }
  TextBlock_19 {
   widgetClassPath=&TextBlock
   namedSlotHost=WBP_HeaderContentPanel_11
   widget=TextBlock_19
   properties {
    text=Filter View
    font {
     bForceMonospaced=false
     bMaterialIsStencil=false
     fontMaterial=null
     fontObject=&Roboto
     letterSpacing=0
     monospacedWidth=1
     size=12
     skewAmount=0
     typefaceFontName=Bold
     outlineSettings {
      bApplyOutlineToDropShadows=false
      bMiteredCorners=false
      bSeparateFillAlpha=false
      outlineMaterial=null
      outlineSize=0
      outlineColor=<rgba 0 0 0 1>
     }
    }
   }
  }
  Border_20 {
   widgetClassPath=&Border
   parent=HorizontalBox_253
   widget=Border_20
   properties {brushColor=<rgba 0.0176 0.0176 0.0176 1>}
   slot {class=HorizontalBoxSlot, properties {size {sizeRule=Fill, value=1}}}
  }
  WBP_SFXGeneratorContentPanel_8 {
   widgetClassPath=&WBP_SFXGeneratorContentPanel
   bIsVariable=true
   parent=Border_20
   widget=WBP_SFXGeneratorContentPanel_8
   slot {class=BorderSlot}
  }
  TextBlock_17 {
   widgetClassPath=&TextBlock
   namedSlotHost=WBP_SFXGeneratorContentPanel_8
   widget=TextBlock_17
   properties {
    text=Modulation
    font {
     bForceMonospaced=false
     bMaterialIsStencil=false
     fontMaterial=null
     fontObject=&Roboto
     letterSpacing=0
     monospacedWidth=1
     size=18
     skewAmount=0
     typefaceFontName=Bold
     outlineSettings {
      bApplyOutlineToDropShadows=false
      bMiteredCorners=false
      bSeparateFillAlpha=false
      outlineMaterial=null
      outlineSize=0
      outlineColor=<rgba 0 0 0 1>
     }
    }
   }
  }
  Border_22 {
   widgetClassPath=&Border
   namedSlotHost=WBP_SFXGeneratorContentPanel_8
   widget=Border_22
   properties {brushColor=<rgba 0.0176 0.0176 0.0176 1>}
  }
  EUW_SFXGenerator_Randomizer_FilterMod {
   widgetClassPath=&EUW_SFXGenerator_Randomizer
   bIsVariable=true
   parent=Border_22
   widget=EUW_SFXGenerator_Randomizer_FilterMod
   slot {class=BorderSlot}
  }
  Border_21 {
   widgetClassPath=&Border
   namedSlotHost=WBP_SFXGeneratorContentPanel_8
   widget=Border_21
   properties {brushColor=<rgba 0.0284 0.0284 0.0284 1>}
  }
  HorizontalBox_252 {
   widgetClassPath=&HorizontalBox
   parent=Border_21
   widget=HorizontalBox_252
   properties {bOverride_Cursor=true}
   slot {class=BorderSlot}
  }
  EUW_SFXGeneratorKnob_Filter_Mod_Depth {
   widgetClassPath=&EUW_MetaSoundFloat_Knob_SFXGen
   bIsVariable=true
   parent=HorizontalBox_252
   widget=EUW_SFXGeneratorKnob_Filter_Mod_Depth
   properties {
    input Name=FX.Filter.Envelope.Depth
    label Text=Depth
    metaSoundLiteralViewModel_Float=&self:WidgetTree.EUW_SFXGeneratorKnob_Filter_Mod_Depth.MetaSoundLiteralViewModel_Float_0
   }
   slot {class=HorizontalBoxSlot, properties {size {sizeRule=Fill, value=1}}}
  }
  EUW_SFXGeneratorKnob_Filter_Mod_Shape {
   widgetClassPath=&EUW_MetaSoundFloat_Knob_SFXGen
   bIsVariable=true
   parent=HorizontalBox_252
   widget=EUW_SFXGeneratorKnob_Filter_Mod_Shape
   properties {
    input Name=FX.Filter.Envelope.Shape
    label Text=Shape
    metaSoundLiteralViewModel_Float=&self:WidgetTree.EUW_SFXGeneratorKnob_Filter_Mod_Shape.MetaSoundLiteralViewModel_Float_0
   }
   slot {class=HorizontalBoxSlot, properties {size {sizeRule=Fill, value=1}}}
  }
  EUW_SFXGeneratorKnob_Filter_Mod_Curve {
   widgetClassPath=&EUW_MetaSoundFloat_Knob_SFXGen
   bIsVariable=true
   parent=HorizontalBox_252
   widget=EUW_SFXGeneratorKnob_Filter_Mod_Curve
   properties {
    input Name=FX.Filter.Envelope.Curve
    label Text=Curve
    metaSoundLiteralViewModel_Float=&self:WidgetTree.EUW_SFXGeneratorKnob_Filter_Mod_Curve.MetaSoundLiteralViewModel_Float_0
   }
   slot {class=HorizontalBoxSlot, properties {size {sizeRule=Fill, value=1}}}
  }
  Border_24 {
   widgetClassPath=&Border
   parent=GridPanel_43
   widget=Border_24
   properties {brushColor=<rgba 0.0176 0.0176 0.0176 1>}
   slot {class=GridSlot, properties {column=2, columnSpan=2, row=1, rowSpan=1}}
  }
  WBP_SFXGeneratorContentPanel_9 {
   widgetClassPath=&WBP_SFXGeneratorContentPanel
   bIsVariable=true
   parent=Border_24
   widget=WBP_SFXGeneratorContentPanel_9
   slot {class=BorderSlot}
  }
  TextBlock_25 {widgetClassPath=&TextBlock, namedSlotHost=WBP_SFXGeneratorContentPanel_9, widget=TextBlock_25, properties {text=Amp}}
  Border_32 {
   widgetClassPath=&Border
   namedSlotHost=WBP_SFXGeneratorContentPanel_9
   widget=Border_32
   properties {brushColor=<rgba 0.0176 0.0176 0.0176 1>}
  }
  EUW_SFXGenerator_Randomizer_Amp {
   widgetClassPath=&EUW_SFXGenerator_Randomizer
   bIsVariable=true
   parent=Border_32
   widget=EUW_SFXGenerator_Randomizer_Amp
   slot {class=BorderSlot}
  }
  Border_25 {
   widgetClassPath=&Border
   namedSlotHost=WBP_SFXGeneratorContentPanel_9
   widget=Border_25
   properties {brushColor=<rgba 0.0284 0.0284 0.0284 1>}
  }
  HorizontalBox_254 {
   widgetClassPath=&HorizontalBox
   parent=Border_25
   widget=HorizontalBox_254
   properties {bOverride_Cursor=true}
   slot {class=BorderSlot}
  }
  Border_29 {
   widgetClassPath=&Border
   parent=HorizontalBox_254
   widget=Border_29
   properties {brushColor=<rgba 0.0176 0.0176 0.0176 1>}
   slot {class=HorizontalBoxSlot}
  }
  WBP_SFXGeneratorContentPanel_11 {
   widgetClassPath=&WBP_SFXGeneratorContentPanel
   bIsVariable=true
   parent=Border_29
   widget=WBP_SFXGeneratorContentPanel_11
   slot {class=BorderSlot}
  }
  TextBlock_24 {
   widgetClassPath=&TextBlock
   namedSlotHost=WBP_SFXGeneratorContentPanel_11
   widget=TextBlock_24
   properties {
    text=Envelope
    font {
     bForceMonospaced=false
     bMaterialIsStencil=false
     fontMaterial=null
     fontObject=&Roboto
     letterSpacing=0
     monospacedWidth=1
     size=18
     skewAmount=0
     typefaceFontName=Bold
     outlineSettings {
      bApplyOutlineToDropShadows=false
      bMiteredCorners=false
      bSeparateFillAlpha=false
      outlineMaterial=null
      outlineSize=0
      outlineColor=<rgba 0 0 0 1>
     }
    }
   }
  }
  Border_31 {
   widgetClassPath=&Border
   namedSlotHost=WBP_SFXGeneratorContentPanel_11
   widget=Border_31
   properties {brushColor=<rgba 0.0176 0.0176 0.0176 1>}
  }
  EUW_SFXGenerator_Randomizer_AmpEnv {
   widgetClassPath=&EUW_SFXGenerator_Randomizer
   bIsVariable=true
   parent=Border_31
   widget=EUW_SFXGenerator_Randomizer_AmpEnv
   slot {class=BorderSlot}
  }
  Border_30 {
   widgetClassPath=&Border
   namedSlotHost=WBP_SFXGeneratorContentPanel_11
   widget=Border_30
   properties {brushColor=<rgba 0.0284 0.0284 0.0284 1>}
  }
  HorizontalBox_256 {
   widgetClassPath=&HorizontalBox
   parent=Border_30
   widget=HorizontalBox_256
   properties {bOverride_Cursor=true}
   slot {class=BorderSlot}
  }
  EUW_SFXGeneratorKnob_AmpEnv_Attack {
   widgetClassPath=&EUW_MetaSoundFloat_Knob_SFXGen
   bIsVariable=true
   parent=HorizontalBox_256
   widget=EUW_SFXGeneratorKnob_AmpEnv_Attack
   properties {
    input Name=Amp.Envelope.Attack
    label Text=Attack
    metaSoundLiteralViewModel_Float=&self:WidgetTree.EUW_SFXGeneratorKnob_AmpEnv_Attack.MetaSoundLiteralViewModel_Float_0
   }
   slot {class=HorizontalBoxSlot}
  }
  EUW_SFXGeneratorKnob_AmpEnv_Decay {
   widgetClassPath=&EUW_MetaSoundFloat_Knob_SFXGen
   bIsVariable=true
   parent=HorizontalBox_256
   widget=EUW_SFXGeneratorKnob_AmpEnv_Decay
   properties {
    input Name=Amp.Envelope.Decay
    label Text=Decay
    metaSoundLiteralViewModel_Float=&self:WidgetTree.EUW_SFXGeneratorKnob_AmpEnv_Decay.MetaSoundLiteralViewModel_Float_0
   }
   slot {class=HorizontalBoxSlot}
  }
  EUW_SFXGeneratorKnob_AmpEnv_Sustain {
   widgetClassPath=&EUW_MetaSoundFloat_Knob_SFXGen
   bIsVariable=true
   parent=HorizontalBox_256
   widget=EUW_SFXGeneratorKnob_AmpEnv_Sustain
   properties {
    input Name=Amp.Envelope.Sustain
    label Text=Sustain
    metaSoundLiteralViewModel_Float=&self:WidgetTree.EUW_SFXGeneratorKnob_AmpEnv_Sustain.MetaSoundLiteralViewModel_Float_0
   }
   slot {class=HorizontalBoxSlot}
  }
  EUW_SFXGeneratorKnob_AmpEnv_Duration {
   widgetClassPath=&EUW_MetaSoundFloat_Knob_SFXGen
   bIsVariable=true
   parent=HorizontalBox_256
   widget=EUW_SFXGeneratorKnob_AmpEnv_Duration
   properties {
    input Name=Amp.Envelope.Duration
    label Text=Duration
    metaSoundLiteralViewModel_Float=&self:WidgetTree.EUW_SFXGeneratorKnob_AmpEnv_Duration.MetaSoundLiteralViewModel_Float_0
   }
   slot {class=HorizontalBoxSlot}
  }
  EUW_SFXGeneratorKnob_AmpEnv_Release {
   widgetClassPath=&EUW_MetaSoundFloat_Knob_SFXGen
   bIsVariable=true
   parent=HorizontalBox_256
   widget=EUW_SFXGeneratorKnob_AmpEnv_Release
   properties {
    input Name=Amp.Envelope.Release
    label Text=Release
    metaSoundLiteralViewModel_Float=&self:WidgetTree.EUW_SFXGeneratorKnob_AmpEnv_Release.MetaSoundLiteralViewModel_Float_0
   }
   slot {class=HorizontalBoxSlot}
  }
  Border_63 {
   widgetClassPath=&Border
   parent=HorizontalBox_254
   widget=Border_63
   properties {brushColor=<rgba 0.0176 0.0176 0.0176 1>}
   slot {class=HorizontalBoxSlot}
  }
  WBP_HeaderContentPanel_12 {
   widgetClassPath=&WBP_HeaderContentPanel
   bIsVariable=true
   parent=Border_63
   widget=WBP_HeaderContentPanel_12
   slot {class=BorderSlot}
  }
  EditorUtilityProgressBar_Env {
   widgetClassPath=/Script/Blutility.EditorUtilityProgressBar
   bIsVariable=true
   namedSlotHost=WBP_HeaderContentPanel_12
   widget=EditorUtilityProgressBar_Env
   properties {
    barFillType=BottomToTop
    fillColorAndOpacity=<rgba 0 0.5 1 1>
    widgetStyle {
     enableFillAnimation=true
     backgroundImage {
      bIsDynamicallyLoaded=false
      drawAs=Box
      imageType=NoImage
      mirroring=NoMirror
      resourceName=None
      resourceObject=null
      tiling=NoTile
      imageSize=<0 0>
      margin=<ltrb 0 0 0 0>
      outlineSettings {
       bUseBrushTransparency=false
       roundingType=HalfHeightRadius
       width=0
       color {colorUseRule=UseColor_Specified, specifiedColor=<rgba 0 0 0 0>}
       cornerRadii=<0 0 0 0>
      }
      tintColor {colorUseRule=UseColor_Specified, specifiedColor=<rgba 0.0103 0.0103 0.0103 1>}
      uVRegion {bIsValid=false, max=<0 0>, min=<0 0>}
     }
     fillImage {
      bIsDynamicallyLoaded=false
      drawAs=Image
      imageType=FullColor
      mirroring=NoMirror
      resourceName=../../../Engine/Content/Slate/Starship/CoreWidgets/ProgressBar/ProgressMarquee.png
      resourceObject=null
      tiling=Horizontal
      imageSize=<20 12>
      margin=<ltrb 0 0 0 0>
      outlineSettings {
       bUseBrushTransparency=false
       roundingType=HalfHeightRadius
       width=0
       color {colorUseRule=UseColor_Specified, specifiedColor=<rgba 0 0 0 0>}
       cornerRadii=<0 0 0 0>
      }
      tintColor {colorUseRule=UseColor_Specified, specifiedColor=<rgba 0.0762 0.6724 0.3813 1>}
      uVRegion {bIsValid=false, max=<0 0>, min=<0 0>}
     }
     marqueeImage {
      bIsDynamicallyLoaded=false
      drawAs=Image
      imageType=FullColor
      mirroring=NoMirror
      resourceName=../../../Engine/Content/Slate/Starship/CoreWidgets/ProgressBar/ProgressMarquee.png
      resourceObject=null
      tiling=Horizontal
      imageSize=<20 12>
      margin=<ltrb 0 0 0 0>
      outlineSettings {
       bUseBrushTransparency=false
       roundingType=HalfHeightRadius
       width=0
       color {colorUseRule=UseColor_Specified, specifiedColor=<rgba 0 0 0 0>}
       cornerRadii=<0 0 0 0>
      }
      tintColor {colorUseRule=UseColor_Specified, specifiedColor=<rgba 0.0044 0.2384 1 1>}
      uVRegion {bIsValid=false, max=<0 0>, min=<0 0>}
     }
    }
   }
  }
  Border_26 {
   widgetClassPath=&Border
   parent=HorizontalBox_254
   widget=Border_26
   properties {brushColor=<rgba 0.0176 0.0176 0.0176 1>}
   slot {class=HorizontalBoxSlot}
  }
  WBP_SFXGeneratorContentPanel_10 {
   widgetClassPath=&WBP_SFXGeneratorContentPanel
   bIsVariable=true
   parent=Border_26
   widget=WBP_SFXGeneratorContentPanel_10
   slot {class=BorderSlot}
  }
  TextBlock_20 {
   widgetClassPath=&TextBlock
   namedSlotHost=WBP_SFXGeneratorContentPanel_10
   widget=TextBlock_20
   properties {
    text=LFO
    font {
     bForceMonospaced=false
     bMaterialIsStencil=false
     fontMaterial=null
     fontObject=&Roboto
     letterSpacing=0
     monospacedWidth=1
     size=18
     skewAmount=0
     typefaceFontName=Bold
     outlineSettings {
      bApplyOutlineToDropShadows=false
      bMiteredCorners=false
      bSeparateFillAlpha=false
      outlineMaterial=null
      outlineSize=0
      outlineColor=<rgba 0 0 0 1>
     }
    }
   }
  }
  Border_28 {
   widgetClassPath=&Border
   namedSlotHost=WBP_SFXGeneratorContentPanel_10
   widget=Border_28
   properties {brushColor=<rgba 0.0176 0.0176 0.0176 1>}
  }
  EUW_SFXGenerator_Randomizer_AmpLFO {
   widgetClassPath=&EUW_SFXGenerator_Randomizer
   bIsVariable=true
   parent=Border_28
   widget=EUW_SFXGenerator_Randomizer_AmpLFO
   slot {class=BorderSlot}
  }
  Border_27 {
   widgetClassPath=&Border
   namedSlotHost=WBP_SFXGeneratorContentPanel_10
   widget=Border_27
   properties {brushColor=<rgba 0.0284 0.0284 0.0284 1>}
  }
  HorizontalBox_255 {
   widgetClassPath=&HorizontalBox
   parent=Border_27
   widget=HorizontalBox_255
   properties {bOverride_Cursor=true}
   slot {class=BorderSlot}
  }
  EUW_SFXGeneratorKnob_AmpLFO_Freq {
   widgetClassPath=&EUW_MetaSoundFloat_Knob_SFXGen
   bIsVariable=true
   parent=HorizontalBox_255
   widget=EUW_SFXGeneratorKnob_AmpLFO_Freq
   properties {
    input Name=Amp.LFO.Frequency
    label Text=Frequency
    metaSoundLiteralViewModel_Float=&self:WidgetTree.EUW_SFXGeneratorKnob_AmpLFO_Freq.MetaSoundLiteralViewModel_Float_0
   }
   slot {class=HorizontalBoxSlot}
  }
  EUW_SFXGeneratorKnob_AmpLFO_Depth {
   widgetClassPath=&EUW_MetaSoundFloat_Knob_SFXGen
   bIsVariable=true
   parent=HorizontalBox_255
   widget=EUW_SFXGeneratorKnob_AmpLFO_Depth
   properties {
    input Name=Amp.LFO.Depth
    label Text=Depth
    metaSoundLiteralViewModel_Float=&self:WidgetTree.EUW_SFXGeneratorKnob_AmpLFO_Depth.MetaSoundLiteralViewModel_Float_0
   }
   slot {class=HorizontalBoxSlot}
  }
  Border_62 {
   widgetClassPath=&Border
   parent=HorizontalBox_254
   widget=Border_62
   properties {brushColor=<rgba 0.0176 0.0176 0.0176 1>}
   slot {class=HorizontalBoxSlot}
  }
  WBP_HeaderContentPanel_419 {
   widgetClassPath=&WBP_HeaderContentPanel
   bIsVariable=true
   parent=Border_62
   widget=WBP_HeaderContentPanel_419
   slot {class=BorderSlot}
  }
  VerticalBox_174 {widgetClassPath=&VerticalBox, namedSlotHost=WBP_HeaderContentPanel_419, widget=VerticalBox_174}
  EditorUtilityProgressBar_LFO {
   widgetClassPath=&EditorUtilityProgressBar
   bIsVariable=true
   parent=VerticalBox_174
   widget=EditorUtilityProgressBar_LFO
   properties {
    barFillType=TopToBottom
    fillColorAndOpacity=<rgba 0 0.5 1 1>
    widgetStyle {
     enableFillAnimation=true
     backgroundImage {
      bIsDynamicallyLoaded=false
      drawAs=Box
      imageType=NoImage
      mirroring=NoMirror
      resourceName=None
      resourceObject=null
      tiling=NoTile
      imageSize=<0 0>
      margin=<ltrb 0 0 0 0>
      outlineSettings {
       bUseBrushTransparency=false
       roundingType=HalfHeightRadius
       width=0
       color {colorUseRule=UseColor_Specified, specifiedColor=<rgba 0 0 0 0>}
       cornerRadii=<0 0 0 0>
      }
      tintColor {colorUseRule=UseColor_Specified, specifiedColor=<rgba 0.0103 0.0103 0.0103 1>}
      uVRegion {bIsValid=false, max=<0 0>, min=<0 0>}
     }
     fillImage {
      bIsDynamicallyLoaded=false
      drawAs=Image
      imageType=FullColor
      mirroring=NoMirror
      resourceName=../../../Engine/Content/Slate/Starship/CoreWidgets/ProgressBar/ProgressMarquee.png
      resourceObject=null
      tiling=Horizontal
      imageSize=<20 12>
      margin=<ltrb 0 0 0 0>
      outlineSettings {
       bUseBrushTransparency=false
       roundingType=HalfHeightRadius
       width=0
       color {colorUseRule=UseColor_Specified, specifiedColor=<rgba 0 0 0 0>}
       cornerRadii=<0 0 0 0>
      }
      tintColor {colorUseRule=UseColor_Specified, specifiedColor=<rgba 0.0908 0.8632 0.4793 1>}
      uVRegion {bIsValid=false, max=<0 0>, min=<0 0>}
     }
     marqueeImage {
      bIsDynamicallyLoaded=false
      drawAs=Image
      imageType=FullColor
      mirroring=NoMirror
      resourceName=../../../Engine/Content/Slate/Starship/CoreWidgets/ProgressBar/ProgressMarquee.png
      resourceObject=null
      tiling=Horizontal
      imageSize=<20 12>
      margin=<ltrb 0 0 0 0>
      outlineSettings {
       bUseBrushTransparency=false
       roundingType=HalfHeightRadius
       width=0
       color {colorUseRule=UseColor_Specified, specifiedColor=<rgba 0 0 0 0>}
       cornerRadii=<0 0 0 0>
      }
      tintColor {colorUseRule=UseColor_Specified, specifiedColor=<rgba 0.0044 0.2384 1 1>}
      uVRegion {bIsValid=false, max=<0 0>, min=<0 0>}
     }
    }
   }
   slot {class=VerticalBoxSlot, properties {size {sizeRule=Fill, value=1}}}
  }
  Border_33 {
   widgetClassPath=&Border
   parent=GridPanel_43
   widget=Border_33
   properties {brushColor=<rgba 0.0176 0.0176 0.0176 1>}
   slot {class=GridSlot, properties {column=4, rowSpan=2}}
  }
  WBP_SFXGeneratorContentPanel_12 {
   widgetClassPath=&WBP_SFXGeneratorContentPanel
   bIsVariable=true
   parent=Border_33
   widget=WBP_SFXGeneratorContentPanel_12
   slot {class=BorderSlot}
  }
  TextBlock_32 {widgetClassPath=&TextBlock, namedSlotHost=WBP_SFXGeneratorContentPanel_12, widget=TextBlock_32, properties {text=Send FX}}
  Border_41 {
   widgetClassPath=&Border
   namedSlotHost=WBP_SFXGeneratorContentPanel_12
   widget=Border_41
   properties {brushColor=<rgba 0.0176 0.0176 0.0176 1>}
  }
  EUW_SFXGenerator_Randomizer_SendFX {
   widgetClassPath=&EUW_SFXGenerator_Randomizer
   bIsVariable=true
   parent=Border_41
   widget=EUW_SFXGenerator_Randomizer_SendFX
   slot {class=BorderSlot}
  }
  Border_34 {
   widgetClassPath=&Border
   namedSlotHost=WBP_SFXGeneratorContentPanel_12
   widget=Border_34
   properties {brushColor=<rgba 0.0284 0.0284 0.0284 1>}
  }
  VerticalBox_11 {widgetClassPath=&VerticalBox, parent=Border_34, widget=VerticalBox_11, slot {class=BorderSlot}}
  HorizontalBox_353 {
   widgetClassPath=&HorizontalBox
   parent=VerticalBox_11
   widget=HorizontalBox_353
   slot {class=VerticalBoxSlot, properties {verticalAlignment=VAlign_Center, size {sizeRule=Fill, value=1}}}
  }
  Border_35 {
   widgetClassPath=&Border
   parent=HorizontalBox_353
   widget=Border_35
   properties {brushColor=<rgba 0.0176 0.0176 0.0176 1>}
   slot {class=HorizontalBoxSlot}
  }
  WBP_SFXGeneratorContentPanel_13 {
   widgetClassPath=&WBP_SFXGeneratorContentPanel
   bIsVariable=true
   parent=Border_35
   widget=WBP_SFXGeneratorContentPanel_13
   slot {class=BorderSlot}
  }
  TextBlock_27 {
   widgetClassPath=&TextBlock
   namedSlotHost=WBP_SFXGeneratorContentPanel_13
   widget=TextBlock_27
   properties {
    text=Quad Flanger
    font {
     bForceMonospaced=false
     bMaterialIsStencil=false
     fontMaterial=null
     fontObject=&Roboto
     letterSpacing=0
     monospacedWidth=1
     size=18
     skewAmount=0
     typefaceFontName=Bold
     outlineSettings {
      bApplyOutlineToDropShadows=false
      bMiteredCorners=false
      bSeparateFillAlpha=false
      outlineMaterial=null
      outlineSize=0
      outlineColor=<rgba 0 0 0 1>
     }
    }
   }
  }
  Border_37 {
   widgetClassPath=&Border
   namedSlotHost=WBP_SFXGeneratorContentPanel_13
   widget=Border_37
   properties {brushColor=<rgba 0.0176 0.0176 0.0176 1>}
  }
  EUW_SFXGenerator_Randomizer_QuadFlanger {
   widgetClassPath=&EUW_SFXGenerator_Randomizer
   bIsVariable=true
   parent=Border_37
   widget=EUW_SFXGenerator_Randomizer_QuadFlanger
   slot {class=BorderSlot}
  }
  HorizontalBox_299 {widgetClassPath=&HorizontalBox, namedSlotHost=WBP_SFXGeneratorContentPanel_13, widget=HorizontalBox_299}
  EUW_SFXGeneratorKnob_SendFX_QuadFlanger_Amount {
   widgetClassPath=&EUW_MetaSoundFloat_Knob_SFXGen
   bIsVariable=true
   parent=HorizontalBox_299
   widget=EUW_SFXGeneratorKnob_SendFX_QuadFlanger_Amount
   properties {
    input Name=FX.Flanger.Amount
    label Text=Amount
    metaSoundLiteralViewModel_Float=&self:WidgetTree.EUW_SFXGeneratorKnob_SendFX_QuadFlanger_Amount.MetaSoundLiteralViewModel_Float_0
   }
   slot {class=HorizontalBoxSlot, properties {size {sizeRule=Fill, value=1}}}
  }
  EUW_SFXGeneratorKnob_SendFX_QuadFlanger_DelayTime {
   widgetClassPath=&EUW_MetaSoundFloat_Knob_SFXGen
   bIsVariable=true
   parent=HorizontalBox_299
   widget=EUW_SFXGeneratorKnob_SendFX_QuadFlanger_DelayTime
   properties {
    input Name=FX.Flanger.Delay
    label Text=Delay Time
    metaSoundLiteralViewModel_Float=&self:WidgetTree.EUW_SFXGeneratorKnob_SendFX_QuadFlanger_DelayTime.MetaSoundLiteralViewModel_Float_0
   }
   slot {class=HorizontalBoxSlot, properties {size {sizeRule=Fill, value=1}}}
  }
  Border_39 {
   widgetClassPath=&Border
   parent=HorizontalBox_353
   widget=Border_39
   properties {brushColor=<rgba 0.0176 0.0176 0.0176 1>}
   slot {class=HorizontalBoxSlot}
  }
  WBP_SFXGeneratorContentPanel_16 {
   widgetClassPath=&WBP_SFXGeneratorContentPanel
   bIsVariable=true
   parent=Border_39
   widget=WBP_SFXGeneratorContentPanel_16
   slot {class=BorderSlot}
  }
  TextBlock_22 {
   widgetClassPath=&TextBlock
   namedSlotHost=WBP_SFXGeneratorContentPanel_16
   widget=TextBlock_22
   properties {
    text=Reverb
    font {
     bForceMonospaced=false
     bMaterialIsStencil=false
     fontMaterial=null
     fontObject=&Roboto
     letterSpacing=0
     monospacedWidth=1
     size=18
     skewAmount=0
     typefaceFontName=Bold
     outlineSettings {
      bApplyOutlineToDropShadows=false
      bMiteredCorners=false
      bSeparateFillAlpha=false
      outlineMaterial=null
      outlineSize=0
      outlineColor=<rgba 0 0 0 1>
     }
    }
   }
  }
  Border_40 {
   widgetClassPath=&Border
   namedSlotHost=WBP_SFXGeneratorContentPanel_16
   widget=Border_40
   properties {brushColor=<rgba 0.0176 0.0176 0.0176 1>}
  }
  SizeBox_52 {
   widgetClassPath=&SizeBox
   parent=Border_40
   widget=SizeBox_52
   properties {bOverride_HeightOverride=true, heightOverride=25}
   slot {class=BorderSlot}
  }
  EUW_SFXGeneratorKnob_SendFX_Reverb_Amount {
   widgetClassPath=&EUW_MetaSoundFloat_Knob_SFXGen
   bIsVariable=true
   namedSlotHost=WBP_SFXGeneratorContentPanel_16
   widget=EUW_SFXGeneratorKnob_SendFX_Reverb_Amount
   properties {
    input Name=FX.Reverb.Amount
    label Text=Amount
    metaSoundLiteralViewModel_Float=&self:WidgetTree.EUW_SFXGeneratorKnob_SendFX_Reverb_Amount.MetaSoundLiteralViewModel_Float_0
   }
  }
  Border_66 {
   widgetClassPath=&Border
   parent=HorizontalBox_353
   widget=Border_66
   properties {brushColor=<rgba 0.0176 0.0176 0.0176 1>}
   slot {class=HorizontalBoxSlot, properties {size {sizeRule=Fill, value=1}}}
  }
  WBP_HeaderContentPanel_14 {
   widgetClassPath=&WBP_HeaderContentPanel
   bIsVariable=true
   parent=Border_66
   widget=WBP_HeaderContentPanel_14
   slot {class=BorderSlot}
  }
  Border_67 {
   widgetClassPath=&Border
   namedSlotHost=WBP_HeaderContentPanel_14
   widget=Border_67
   properties {brushColor=<rgba 0.0176 0.0176 0.0176 1>}
  }
  SizeBox_9 {
   widgetClassPath=&SizeBox
   parent=Border_67
   widget=SizeBox_9
   properties {bOverride_HeightOverride=true, bOverride_WidthOverride=true, widthOverride=150}
   slot {class=BorderSlot}
  }
  AudioOscilloscope_7 {
   widgetClassPath=&AudioOscilloscope
   bIsVariable=true
   parent=SizeBox_9
   widget=AudioOscilloscope_7
   properties {analysisPeriodMs=20, audioBus=&AB_SFXG_Reverb, maxTimeWindowMs=20, timeWindowMs=50}
   slot {class=SizeBoxSlot}
  }
  TextBlock_26 {
   widgetClassPath=&TextBlock
   namedSlotHost=WBP_HeaderContentPanel_14
   widget=TextBlock_26
   properties {
    text=Reverb View
    font {
     bForceMonospaced=false
     bMaterialIsStencil=false
     fontMaterial=null
     fontObject=&Roboto
     letterSpacing=0
     monospacedWidth=1
     size=12
     skewAmount=0
     typefaceFontName=Bold
     outlineSettings {
      bApplyOutlineToDropShadows=false
      bMiteredCorners=false
      bSeparateFillAlpha=false
      outlineMaterial=null
      outlineSize=0
      outlineColor=<rgba 0 0 0 1>
     }
    }
   }
  }
  HorizontalBox_354 {
   widgetClassPath=&HorizontalBox
   parent=VerticalBox_11
   widget=HorizontalBox_354
   slot {class=VerticalBoxSlot, properties {verticalAlignment=VAlign_Center, size {sizeRule=Fill, value=1}}}
  }
  Border_36 {
   widgetClassPath=&Border
   parent=HorizontalBox_354
   widget=Border_36
   properties {brushColor=<rgba 0.0176 0.0176 0.0176 1>}
   slot {class=HorizontalBoxSlot}
  }
  WBP_SFXGeneratorContentPanel_14 {
   widgetClassPath=&WBP_SFXGeneratorContentPanel
   bIsVariable=true
   parent=Border_36
   widget=WBP_SFXGeneratorContentPanel_14
   slot {class=BorderSlot}
  }
  TextBlock_21 {
   widgetClassPath=&TextBlock
   namedSlotHost=WBP_SFXGeneratorContentPanel_14
   widget=TextBlock_21
   properties {
    text=Delay
    font {
     bForceMonospaced=false
     bMaterialIsStencil=false
     fontMaterial=null
     fontObject=&Roboto
     letterSpacing=0
     monospacedWidth=1
     size=18
     skewAmount=0
     typefaceFontName=Bold
     outlineSettings {
      bApplyOutlineToDropShadows=false
      bMiteredCorners=false
      bSeparateFillAlpha=false
      outlineMaterial=null
      outlineSize=0
      outlineColor=<rgba 0 0 0 1>
     }
    }
   }
  }
  Border_38 {
   widgetClassPath=&Border
   namedSlotHost=WBP_SFXGeneratorContentPanel_14
   widget=Border_38
   properties {brushColor=<rgba 0.0176 0.0176 0.0176 1>}
  }
  EUW_SFXGenerator_Randomizer_Delay {
   widgetClassPath=&EUW_SFXGenerator_Randomizer
   bIsVariable=true
   parent=Border_38
   widget=EUW_SFXGenerator_Randomizer_Delay
   slot {class=BorderSlot}
  }
  HorizontalBox_352 {widgetClassPath=&HorizontalBox, namedSlotHost=WBP_SFXGeneratorContentPanel_14, widget=HorizontalBox_352}
  EUW_SFXGeneratorKnob_SendFX_Delay_Amount {
   widgetClassPath=&EUW_MetaSoundFloat_Knob_SFXGen
   bIsVariable=true
   parent=HorizontalBox_352
   widget=EUW_SFXGeneratorKnob_SendFX_Delay_Amount
   properties {
    input Name=FX.Delay.Amount
    label Text=Amount
    metaSoundLiteralViewModel_Float=&self:WidgetTree.EUW_SFXGeneratorKnob_SendFX_Delay_Amount.MetaSoundLiteralViewModel_Float_0
   }
   slot {class=HorizontalBoxSlot, properties {size {sizeRule=Fill, value=1}}}
  }
  EUW_SFXGeneratorKnob_SendFX_Delay_Time {
   widgetClassPath=&EUW_MetaSoundFloat_Knob_SFXGen
   bIsVariable=true
   parent=HorizontalBox_352
   widget=EUW_SFXGeneratorKnob_SendFX_Delay_Time
   properties {
    input Name=FX.Delay.Time
    label Text=Time
    metaSoundLiteralViewModel_Float=&self:WidgetTree.EUW_SFXGeneratorKnob_SendFX_Delay_Time.MetaSoundLiteralViewModel_Float_0
   }
   slot {class=HorizontalBoxSlot, properties {size {sizeRule=Fill, value=1}}}
  }
  EUW_SFXGeneratorKnob_SendFX_Delay_Feedback {
   widgetClassPath=&EUW_MetaSoundFloat_Knob_SFXGen
   bIsVariable=true
   parent=HorizontalBox_352
   widget=EUW_SFXGeneratorKnob_SendFX_Delay_Feedback
   properties {
    input Name=FX.Delay.Feedback
    label Text=Feedback
    metaSoundLiteralViewModel_Float=&self:WidgetTree.EUW_SFXGeneratorKnob_SendFX_Delay_Feedback.MetaSoundLiteralViewModel_Float_0
   }
   slot {class=HorizontalBoxSlot, properties {size {sizeRule=Fill, value=1}}}
  }
  Border_68 {
   widgetClassPath=&Border
   parent=HorizontalBox_354
   widget=Border_68
   properties {brushColor=<rgba 0.0176 0.0176 0.0176 1>}
   slot {class=HorizontalBoxSlot, properties {size {sizeRule=Fill, value=1}}}
  }
  WBP_HeaderContentPanel_15 {
   widgetClassPath=&WBP_HeaderContentPanel
   bIsVariable=true
   parent=Border_68
   widget=WBP_HeaderContentPanel_15
   slot {class=BorderSlot}
  }
  Border_69 {
   widgetClassPath=&Border
   namedSlotHost=WBP_HeaderContentPanel_15
   widget=Border_69
   properties {brushColor=<rgba 0.0176 0.0176 0.0176 1>}
  }
  SizeBox_10 {
   widgetClassPath=&SizeBox
   parent=Border_69
   widget=SizeBox_10
   properties {bOverride_HeightOverride=true, bOverride_WidthOverride=true, widthOverride=150}
   slot {class=BorderSlot}
  }
  AudioOscilloscope_8 {
   widgetClassPath=&AudioOscilloscope
   bIsVariable=true
   parent=SizeBox_10
   widget=AudioOscilloscope_8
   properties {analysisPeriodMs=20, audioBus=&AB_SFXG_Delay, maxTimeWindowMs=20, timeWindowMs=50}
   slot {class=SizeBoxSlot}
  }
  TextBlock_28 {
   widgetClassPath=&TextBlock
   namedSlotHost=WBP_HeaderContentPanel_15
   widget=TextBlock_28
   properties {
    text=Delay View
    font {
     bForceMonospaced=false
     bMaterialIsStencil=false
     fontMaterial=null
     fontObject=&Roboto
     letterSpacing=0
     monospacedWidth=1
     size=12
     skewAmount=0
     typefaceFontName=Bold
     outlineSettings {
      bApplyOutlineToDropShadows=false
      bMiteredCorners=false
      bSeparateFillAlpha=false
      outlineMaterial=null
      outlineSize=0
      outlineColor=<rgba 0 0 0 1>
     }
    }
   }
  }
 ]
}
